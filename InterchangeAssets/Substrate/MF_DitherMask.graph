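# Substrate material function that turns a transmission value into a dithered, view-dependent mask, output as an opacity mask and as a weight on an input BSDF.
# Interleaved gradient golden noise is centred on zero, scaled by the dither input and added to transmission.
# The sum drives a Fresnel node's base reflect fraction with exponent one.
# The result feeds both the Substrate weight and the Opacity Mask output.
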MaterialFunction /InterchangeAssets/Substrate/MF_DitherMask
graph {
 properties {bExposeToLibrary=true}
 nodes [
  n1 = SubstrateSlabBSDF!(
   Diffuse Albedo
   F0
   F90
   Roughness
   Anisotropy
   Normal
   Tangent
   SSS MFP
   SSS MFP Scale
   SSS Phase Anisotropy
   Emissive Color
   Second Roughness
   Second Roughness Weight
   Fuzz Roughness
   Fuzz Amount
   Fuzz Color
   "Glint Density (Disabled)"
   "Glint UVs (Disabled)"
  )
  n2 = FunctionInput() {bUsePreviewValueAsDefault=true, inputName=transmission, inputType=FunctionInput_Scalar, sortPriority=1, previewValue=<1 0 0 1>}
  n3 = Subtract(A=MaterialFunctionCall!(TemporalSampleIndex, PixelPosition) {materialFunction=/BaseMaterial/Materials/Functions/MF_Noise_InterleavedGradientGolden_1d}.Unsigned) {constB=0.5}
  n4 = FunctionInput() {bUsePreviewValueAsDefault=true, inputName=dither, inputType=FunctionInput_Scalar, sortPriority=2, previewValue=<4 0 0 1>}
  n5 = Fresnel(ExponentIn=Constant() {r=1}, BaseReflectFractionIn=Add(A=@n2, B=Multiply(A=@n3, B=@n4)))
  FunctionOutput(None=SubstrateWeight!(A=FunctionInput(Preview=@n1) {bUsePreviewValueAsDefault=true, inputName=BSDF, inputType=FunctionInput_Substrate}, Weight=@n5)) {outputName=Dither, sortPriority=1}
  FunctionOutput(None=@n5) {outputName=Opacity Mask}
 ]
}
