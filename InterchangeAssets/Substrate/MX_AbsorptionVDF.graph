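# MaterialX material function that constructs a VDF for a purely light-absorbing medium from a per-channel absorption rate in inverse meters.
# The absorption input serves directly as Extinction, and 100 divided by it gives the mean free path (MFP), apparently in centimeters.
# SetMaterialAttributes packs Extinction into Specular and MFP into Metallic, the reverse of what the function's description states.
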MaterialFunction /InterchangeAssets/Substrate/MX_AbsorptionVDF
graph {
 properties {
  description=Constructs a VDF for pure light absorption. MFP is stored in the Specular and Extinction is stored in Metallic.
  libraryCategoriesText [MaterialX]
 }
 nodes [
  n1 = FunctionInput() {
   bUsePreviewValueAsDefault=true
   description="Absorption rate for the medium (rate per distance traveled in themedium, given in m−1 ). Set for each color component/wavelength separately."
   inputName=absorption
   previewValue=<0 0 0 0>
  }
  n2 = NamedRerouteDeclaration(None=@n1) {name=Extinction, nodeColor=<rgba 0.1484 0 1 1>}
  n3 = Divide(B=@n2) {constA=100}
  FunctionOutput(None=SetMaterialAttributes!(MaterialAttributes, Specular=@n2, Metallic=NamedRerouteDeclaration(None=@n3) {name=MFP, nodeColor=<rgba 1 0.4688 0 1>}) {attributeSetTypes [#1, #2]}) {outputName=Output}
 ]
}
