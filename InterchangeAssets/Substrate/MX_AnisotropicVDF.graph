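# MX_AnisotropicVDF is a MaterialX material function that describes a Henyey-Greenstein volume scattering medium for Interchange Substrate materials.
# It sums absorption and scattering into extinction, and derives albedo and mean free path (MFP) by division, with 100 as the MFP numerator.
# The output repurposes material attribute channels as carriers: albedo in Base Color, MFP in Metallic, extinction in Specular, with anisotropy passed straight to Anisotropy.
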MaterialFunction /InterchangeAssets/Substrate/MX_AnisotropicVDF
graph {
 properties {
  description="Constructs a VDF scattering light for a participating medium, based onthe Henyey-Greenstein phase function. Forward, backward and uniform scattering is supportedand controlled by the anisotropy input. Albedo is stored in Base Color, MFP in Metallic, Extinction in Specular and Anisotropy in Anisotropy"
  libraryCategoriesText [MaterialX]
 }
 nodes [
  n1 = FunctionInput() {
   bUsePreviewValueAsDefault=true
   description="Scattering rate for the medium (rate per distance traveled in themedium, given in m−1 ). Set for each color component/wavelength separately."
   inputName=scattering
   sortPriority=1
  }
  n2 = NamedRerouteUsage()
  n3 = NamedRerouteUsage()
  n4 = FunctionInput() {
   bUsePreviewValueAsDefault=true
   description="Absorption rate for the medium (rate per distance traveled in themedium, given in m−1 ). Set for each color component/wavelength separately."
   inputName=absorption
  }
  n5 = FunctionInput() {
   bUsePreviewValueAsDefault=true
   description="Anisotropy factor, controlling the scattering direction, range [-1.0,1.0]. Negative values give backwards scattering, positive values give forward scattering, and a value of 0.0 (the default) gives uniform scattering."
   inputName=anisotropy
   inputType=FunctionInput_Scalar
   sortPriority=2
  }
  n6 = SetMaterialAttributes!(
   MaterialAttributes
   BaseColor=NamedRerouteDeclaration(None=Divide(A=@n1, B=@n2) {constB=100}) {name=Albedo, nodeColor=<rgba 1 0.8672 0 1>}
   Metallic=NamedRerouteDeclaration(None=Divide(B=@n3) {constA=100}) {name=MFP, nodeColor=<rgba 0 0.8125 1 1>}
   Specular=NamedRerouteDeclaration(None=Add(A=@n4, B=@n1)) {name=Extinction, nodeColor=<rgba 0.5469 0 1 1>}
   Anisotropy=@n5
  ) {attributeSetTypes [
    #1
    #2
    #3
    #4
   ]}
  FunctionOutput(None=@n6) {outputName=Output}
 ]
}
