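# Material function for Interchange that stands in for the MaterialX Chiang hair BSDF under Substrate.
# It declares that model's inputs: R, TT and TRT tints and roughness pairs, IOR, cuticle angle, absorption, normal and curve direction.
# None of these inputs connects to anything; the output is a bare Substrate Hair BSDF with unwired pins.
# It is therefore an interface placeholder, and caller values do not affect the shading.
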MaterialFunction /InterchangeAssets/Substrate/MX_ChiangHairBSDF
graph {nodes [
  FunctionInput() {bUsePreviewValueAsDefault=true, description=Color multiplier for the first R-lobe., inputName=tint_R, previewValue=<1 1 1 1>}
  FunctionInput() {
   bUsePreviewValueAsDefault=true
   description=Color multiplier for the first TT-lobe.
   inputName=tint_TT
   sortPriority=1
   previewValue=<1 1 1 1>
  }
  FunctionInput(Preview=VertexTangentWS()) {
   bUsePreviewValueAsDefault=true
   description=Direction of the hair geometry. Defaults to world space tangent.
   inputName=curve_direction
   sortPriority=10
   previewValue=<0 0 0 0>
  }
  FunctionInput() {
   bUsePreviewValueAsDefault=true
   description=Color multiplier for the first TRT-lobe.
   inputName=tint_TRT
   sortPriority=2
   previewValue=<1 1 1 1>
  }
  FunctionInput() {
   bUsePreviewValueAsDefault=true
   description=Index of refraction. Defaults to 1.55 being the value for keratin.
   inputName=ior
   inputType=FunctionInput_Scalar
   sortPriority=3
   previewValue=<1.55 1 1 1>
  }
  FunctionInput() {
   bUsePreviewValueAsDefault=true
   description="Longitudinal and azimuthal roughness (ν, s) for the first R-lobe, range [0.0, ∞). With (0, 0) specifying pure specular scattering."
   inputName=roughness_R
   inputType=FunctionInput_Vector2
   sortPriority=4
   previewValue=<0.1 0.1 1 1>
  }
  FunctionInput() {
   bUsePreviewValueAsDefault=true
   description="Longitudinal and azimuthal roughness (ν, s) for the first TT-lobe, range [0.0, ∞). With (0, 0) specifying pure specular scattering."
   inputName=roughness_TT
   inputType=FunctionInput_Vector2
   sortPriority=5
   previewValue=<0.05 0.05 1 1>
  }
  FunctionInput() {
   bUsePreviewValueAsDefault=true
   description="Longitudinal and azimuthal roughness (ν, s) for the first TRT-lobe, range [0.0, ∞). With (0, 0) specifying pure specular scattering."
   inputName=roughness_TRT
   inputType=FunctionInput_Vector2
   sortPriority=6
   previewValue=<0.2 0.2 1 1>
  }
  FunctionInput() {
   bUsePreviewValueAsDefault=true
   description="Cuticle angle in radians, Values above 0.5 tilt the scales towards the root of the fiber, range [0.0, 1.0]. With 0.5 specifying no tilt. "
   inputName=cuticle_angle
   inputType=FunctionInput_Scalar
   sortPriority=7
   previewValue=<0.5 0.5 0.5 0.5>
  }
  FunctionInput() {
   bUsePreviewValueAsDefault=true
   description=Absorption coefficient normalized to the hair fiber diameter.
   inputName=absorption_coefficient
   sortPriority=8
   previewValue=<0 0 0 0>
  }
  FunctionInput(Preview=VertexNormalWS()) {
   bUsePreviewValueAsDefault=true
   description=Normal vector of the surface. Defaults to world space normal.
   inputName=normal
   sortPriority=9
   previewValue=<0 0 0 0>
  }
  FunctionOutput(None=SubstrateHairBSDF!(BaseColor, Scatter, Specular, Roughness, Backlit, Tangent, EmissiveColor)) {outputName=Output}
 ]}
