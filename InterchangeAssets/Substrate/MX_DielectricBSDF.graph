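# Material function in the MaterialX library category that stands in for the MaterialX dielectric BSDF as a Substrate material.
# It declares the node's inputs, including weight, tint, IOR, roughness, thin film, normal, tangent, distribution and scatter mode.
# None of these inputs are wired: the output is an unconnected Substrate Slab BSDF, so the described reflection, transmission and layering behavior is not implemented here.
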MaterialFunction /InterchangeAssets/Substrate/MX_DielectricBSDF
graph {
 properties {
  description="Constructs a reflection and/or transmission BSDF based on a microfacet reflectance model and a Fresnel curve for dielectrics. If reflection scattering is enabled the node may be layered vertically over a base BSDF for the surface beneath the dielectric layer. By chaining multiple <dielectric_bsdf> nodes you can describe a surface with multiple specular lobes. If transmission scattering is enabled the node may be layered over a VDF describing the surface interior to handle absorption and scattering inside the medium, useful for colored glass, turbid water, etc."
  libraryCategoriesText [MaterialX]
 }
 nodes [
  FunctionInput() {
   bUsePreviewValueAsDefault=true
   description="Weight for this BSDF’s contribution"
   inputName=weight
   inputType=FunctionInput_Scalar
   previewValue=<1 0 0 1>
  }
  FunctionInput() {
   bUsePreviewValueAsDefault=true
   description=Color weight to tint the reflected and transmitted light. Note that changing the tint gives non-physical results and should only be done when needed for artistic purposes.
   inputName=tint
   sortPriority=1
   previewValue=<1 1 1 1>
  }
  FunctionInput() {
   bUsePreviewValueAsDefault=true
   description=Index of refraction of the surface. Defaults to 1.5. If set to 0.0 the Fresnel curveis disabled and reflectivity is controlled only by weight and tint.
   inputName=ior
   inputType=FunctionInput_Scalar
   sortPriority=2
   previewValue=<1.5 1.5 1.5 1.5>
  }
  FunctionInput(Preview=Constant3Vector() {constant=<rgba 0 0 1 0>}) {description=Normal vector of the surface., inputName=normal, sortPriority=6, previewValue=<0.18 0.18 0.18 0.18>}
  FunctionInput() {
   bUsePreviewValueAsDefault=true
   description="The thickness of an iridescent thin film layer[^Belcour2017] applied over the base bsdf, expressed in nanometers."
   inputName=thinfilm_thickness
   inputType=FunctionInput_Scalar
   sortPriority=4
  }
  FunctionInput() {
   bUsePreviewValueAsDefault=true
   description=Surface roughness.
   inputName=roughness
   inputType=FunctionInput_Vector2
   sortPriority=3
   previewValue=<0.05 0.05 0.05 0.05>
  }
  FunctionInput(Preview=Constant3Vector() {constant=<rgba 0 0 1 0>}) {description=Tangent vector of the surface., inputName=tangent, sortPriority=7, previewValue=<0.18 0.18 0.18 0.18>}
  FunctionInput() {
   bUsePreviewValueAsDefault=true
   description=The index of refraction of the thin film layer.
   inputName=thinfilm_ior
   inputType=FunctionInput_Scalar
   sortPriority=5
   previewValue=<1.5 0 0 1>
  }
  FunctionInput() {
   bUsePreviewValueAsDefault=true
   description=Microfacet distribution type. Defaults to ggx.
   inputName=distribution
   inputType=FunctionInput_Scalar
   sortPriority=8
  }
  FunctionInput() {
   bUsePreviewValueAsDefault=true
   description="Scattering mode, specifying whether the BSDF supports reflection R, transmission T or both reflection and transmission RT. With RT, reflection and transmission occur both when entering and leaving a surface, with their respective intensities controlled by the Fresnel curve. Depending on the IOR and incident angle, it is possible for total internal reflection to occur, generating no transmission even if T or RT is selected. Defaults to R. 0: R, 1: T, 2: RT"
   inputName=scatter_mode
   inputType=FunctionInput_Scalar
   sortPriority=9
  }
  MaterialExpressionStrataSlabBSDF_0 = SubstrateSlabBSDF!(
   Diffuse Albedo
   F0
   F90
   Roughness
   Anisotropy
   Normal
   Tangent
   SSS MFP
   SSS MFP Scale
   SSS Phase Anisotropy
   Emissive Color
   Second Roughness
   Second Roughness Weight
   Fuzz Roughness
   Fuzz Amount
   Fuzz Color
   "Glint Density (Disabled)"
   "Glint UVs (Disabled)"
  )
  FunctionOutput(None=@MaterialExpressionStrataSlabBSDF_0) {outputName=Output}
 ]
}
