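# Material function in the MaterialX library that stands in for the generalized Schlick BSDF node, outputting a Substrate slab.
# It declares inputs for weight, color0, color90, exponent, two-component roughness, normal and tangent, with preview defaults.
# None of these inputs are wired to the Slab BSDF, so the output is an unconfigured slab.
# The reflection, transmission and layering behavior in its description is not built in this graph.
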MaterialFunction /InterchangeAssets/Substrate/MX_GeneralizedSchlickBSDF
graph {
 properties {
  description="Constructs a reflection and/or transmission BSDF based on a microfacet model and a generalized Schlick Fresnel curve. If reflection scattering is enabled the node may be layered vertically over a base BSDF for the surface beneath the dielectric layer. By chaining multiple <generalized_schlick_bsdf> nodes you can describe a surface with multiple specular lobes. If transmission scattering is enabled the node may be layered over a VDF describing the surface interior to handle absorption and scattering inside the medium, useful for colored glass, turbid water, etc."
  libraryCategoriesText [MaterialX]
 }
 nodes [
  FunctionInput() {bUsePreviewValueAsDefault=true, inputName=weight, inputType=FunctionInput_Scalar, previewValue=<1 0 0 1>}
  FunctionInput() {
   bUsePreviewValueAsDefault=true
   description=Reflectivity per color component at facing angles.
   inputName=color0
   sortPriority=1
   previewValue=<1 1 1 1>
  }
  FunctionInput() {
   bUsePreviewValueAsDefault=true
   description=Surface roughness.
   inputName=roughness
   inputType=FunctionInput_Vector2
   sortPriority=4
   previewValue=<0 0 0 0>
  }
  FunctionInput(Preview=Constant3Vector() {constant=<rgba 0 0 1 0>}) {description=Normal vector of the surface., inputName=normal, sortPriority=5, previewValue=<0.18 0.18 0.18 0.18>}
  FunctionInput() {
   bUsePreviewValueAsDefault=true
   description=Reflectivity per color component at grazing angles.
   inputName=color90
   sortPriority=2
   previewValue=<1 1 1 1>
  }
  FunctionInput() {
   bUsePreviewValueAsDefault=true
   description=Exponent for the Schlick blending between color0 and color90.
   inputName=exponent
   inputType=FunctionInput_Scalar
   sortPriority=3
   previewValue=<5 5 5 5>
  }
  FunctionInput(Preview=Constant3Vector() {constant=<rgba 0 0 1 0>}) {description=Tangent vector of the surface., inputName=tangent, sortPriority=6, previewValue=<0.18 0.18 0.18 0.18>}
  MaterialExpressionStrataSlabBSDF_0 = SubstrateSlabBSDF!(
   Diffuse Albedo
   F0
   F90
   Roughness
   Anisotropy
   Normal
   Tangent
   SSS MFP
   SSS MFP Scale
   SSS Phase Anisotropy
   Emissive Color
   Second Roughness
   Second Roughness Weight
   Fuzz Roughness
   Fuzz Amount
   Fuzz Color
   "Glint Density (Disabled)"
   "Glint UVs (Disabled)"
  )
  FunctionOutput(None=@MaterialExpressionStrataSlabBSDF_0) {outputName=Output}
 ]
}
