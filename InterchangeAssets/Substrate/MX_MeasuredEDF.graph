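# MX_MeasuredEDF is a MaterialX library material function meant to construct a Substrate EDF that emits light in a cone around the surface normal.
# It declares color, normal and IES profile texture inputs, with normal defaulting to a preview vector.
# None of these inputs are wired in: the output is a Substrate Slab BSDF with unconnected pins, so the function acts as a placeholder.
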MaterialFunction /InterchangeAssets/Substrate/MX_MeasuredEDF
graph {
 properties {description=Constructs an EDF emitting light inside a cone around the normal direction., libraryCategoriesText [MaterialX]}
 nodes [
  FunctionInput() {description=Radiant emittance of light leaving the surface., inputName=color, previewValue=<1 1 1 1>}
  FunctionInput(Preview=Constant3Vector() {constant=<rgba 0 0 1 0>}) {bUsePreviewValueAsDefault=true, description=Normal vector of the surface., inputName=normal, sortPriority=1}
  FunctionInput() {description=Path to a file containing IES light profile data., inputName=file, inputType=FunctionInput_Texture2D, sortPriority=2}
  MaterialExpressionStrataSlabBSDF_0 = SubstrateSlabBSDF!(
   Diffuse Albedo
   F0
   F90
   Roughness
   Anisotropy
   Normal
   Tangent
   SSS MFP
   SSS MFP Scale
   SSS Phase Anisotropy
   Emissive Color
   Second Roughness
   Second Roughness Weight
   Fuzz Roughness
   Fuzz Amount
   Fuzz Color
   "Glint Density (Disabled)"
   "Glint UVs (Disabled)"
  )
  FunctionOutput(None=@MaterialExpressionStrataSlabBSDF_0) {outputName=Output}
 ]
}
