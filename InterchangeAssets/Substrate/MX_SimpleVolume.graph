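# MX_SimpleVolume is a MaterialX Substrate library function that constructs a volume from a volume distribution (vdf) and an emission (edf) input.
# It unpacks the vdf material attributes to feed a Substrate Slab's diffuse albedo, SSS mean free path and phase anisotropy, while edf drives emissive color.
# The slab uses the Simple Volume subsurface type and is returned as the Volume output.
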MaterialFunction /InterchangeAssets/Substrate/MX_SimpleVolume
graph {
 properties {bExposeToLibrary=true, description=A constructor node for the surfaceshader type., libraryCategoriesText [MaterialX Substrate]}
 nodes [
  n1 = GetMaterialAttributes(None=FunctionInput() {inputName=vdf, inputType=FunctionInput_MaterialAttributes}) {attributeGetTypes [#1, #2, #3]}
  FunctionOutput(None=SubstrateSlabBSDF!(
    Diffuse Albedo=@n1.BaseColor
    F0
    F90
    Roughness
    Anisotropy
    Normal
    Tangent
    SSS MFP=@n1.BaseColor
    SSS MFP Scale
    SSS Phase Anisotropy=@n1.BaseColor
    Emissive Color=FunctionInput() {bUsePreviewValueAsDefault=true, inputName=edf, sortPriority=1}
    Second Roughness
    Second Roughness Weight
    Fuzz Roughness
    Fuzz Amount
    Fuzz Color
    "Glint Density (Disabled)"
    "Glint UVs (Disabled)"
   ) {subSurfaceType=MSS_SimpleVolume}) {outputName=Volume, sortPriority=1}
 ]
}
