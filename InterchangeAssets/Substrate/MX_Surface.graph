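# MX_Surface is a Substrate material function in the MaterialX Substrate library, serving as the constructor node for the MaterialX surfaceshader type.
# It takes bsdf, edf and opacity inputs, with default Substrate Slab BSDFs as previews for the first two and opacity defaulting to 1.
# The bsdf passes straight to the Surface output and opacity to the Opacity output.
# The edf input is declared but feeds no output, so emission is ignored.
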MaterialFunction /InterchangeAssets/Substrate/MX_Surface
graph {
 properties {bExposeToLibrary=true, description=A constructor node for the surfaceshader type., libraryCategoriesText [MaterialX Substrate]}
 nodes [
  MaterialExpressionStrataSlabBSDF_0 = SubstrateSlabBSDF!(
   Diffuse Albedo
   F0
   F90
   Roughness
   Anisotropy
   Normal
   Tangent
   SSS MFP
   SSS MFP Scale
   SSS Phase Anisotropy
   Emissive Color
   Second Roughness
   Second Roughness Weight
   Fuzz Roughness
   Fuzz Amount
   Fuzz Color
   "Glint Density (Disabled)"
   "Glint UVs (Disabled)"
  )
  FunctionInput(Preview=@MaterialExpressionStrataSlabBSDF_0) {
   bUsePreviewValueAsDefault=true
   description=Distribution function for surface emission.
   inputName=edf
   inputType=FunctionInput_Substrate
   sortPriority=1
  }
  MaterialExpressionStrataSlabBSDF_2 = SubstrateSlabBSDF()
  n1 = FunctionInput(Preview=@MaterialExpressionStrataSlabBSDF_2) {bUsePreviewValueAsDefault=true, description=Distribution function for surface scattering., inputName=bsdf, inputType=FunctionInput_Substrate}
  FunctionOutput(None=@n1) {outputName=Surface}
  n2 = FunctionInput() {
   bUsePreviewValueAsDefault=true
   description=Surface cutout opacity
   inputName=opacity
   inputType=FunctionInput_Scalar
   sortPriority=2
   previewValue=<1 0 0 1>
  }
  FunctionOutput(None=@n2) {outputName=Opacity}
 ]
}
