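# Material function in the MaterialX Substrate library that implements an unlit surface for Interchange using a Substrate Unlit BSDF.
# Emission and transmission are each an amount times a color, shared through named reroutes and intended for the BSDF's emissive and transmittance inputs.
# It outputs the BSDF as Surface Unlit, plus a separate OpacityMask for cutout opacity.
# An Opacity reroute also averages a lerp from one into a scalar via a dot product.
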MaterialFunction /InterchangeAssets/Substrate/MX_SurfaceUnlit
graph {
 properties {bExposeToLibrary=true, libraryCategoriesText [MaterialX Substrate]}
 nodes [
  MaterialExpressionStrataUnlitBSDF_0 = SubstrateUnlitBSDF!(EmissiveColor=NamedRerouteUsage(), TransmittanceColor=NamedRerouteUsage(), Normal)
  FunctionOutput(None=@MaterialExpressionStrataUnlitBSDF_0) {outputName=Surface Unlit, sortPriority=1}
  FunctionOutput(None=NamedRerouteUsage()) {outputName=OpacityMask}
  n1 = FunctionInput() {
   bUsePreviewValueAsDefault=true
   description=Surface transmission amount.
   inputName=transmission
   inputType=FunctionInput_Scalar
   sortPriority=2
  }
  n2 = FunctionInput() {
   bUsePreviewValueAsDefault=true
   description=Surface transmission color.
   inputName=transmission_color
   sortPriority=3
   previewValue=<1 1 1 1>
  }
  NamedRerouteDeclaration(None=Multiply(A=@n1, B=@n2)) {name=Transmission, nodeColor=<rgba 0.4531 1 0 1>}
  n3 = FunctionInput() {
   bUsePreviewValueAsDefault=true
   description=Surface cutout opacity.
   inputName=opacity
   inputType=FunctionInput_Scalar
   previewValue=<1 1 1 1>
  }
  NamedRerouteDeclaration(None=@n3) {name=OpacityMask, nodeColor=<rgba 0.1016 0 1 1>}
  n4 = NamedRerouteUsage()
  n5 = NamedRerouteUsage()
  NamedRerouteDeclaration(None=DotProduct(A=LinearInterpolate(B=@n4, Alpha=@n5) {constA=1}, B=Constant() {r=0.3333})) {name=Opacity, nodeColor=<rgba 1 0 0.8906 1>}
  n6 = FunctionInput() {
   bUsePreviewValueAsDefault=true
   description=Surface emission amount.
   inputName=emission
   inputType=FunctionInput_Scalar
   previewValue=<1 1 1 1>
  }
  n7 = FunctionInput() {bUsePreviewValueAsDefault=true, description=Surface emission color., inputName=emission_color, sortPriority=1, previewValue=<1 1 1 1>}
  NamedRerouteDeclaration(None=Multiply(A=@n6, B=@n7)) {name=Emission, nodeColor=<rgba 1 0 0.6562 1>}
 ]
}
