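# MX_ThinFilmBSDF is a Substrate material function in the MaterialX library that stands in for a thin-film BSDF, marked deprecated since MaterialX 1.39.3.
# It declares scalar thickness and ior inputs with preview values as defaults, but neither is wired to anything.
# The output is an unconnected Substrate Slab BSDF at its default settings, so the function computes no thin-film effect.
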MaterialFunction /InterchangeAssets/Substrate/MX_ThinFilmBSDF
graph {
 properties {description=DEPRECATED since MaterialX 1.39.3, libraryCategoriesText [MaterialX]}
 nodes [
  FunctionInput() {bUsePreviewValueAsDefault=true, inputName=thickness, inputType=FunctionInput_Scalar, previewValue=<550 550 550 550>}
  FunctionInput() {bUsePreviewValueAsDefault=true, inputName=ior, inputType=FunctionInput_Scalar, previewValue=<1.5 1.5 1.5 1.5>}
  MaterialExpressionStrataSlabBSDF_0 = SubstrateSlabBSDF!(
   Diffuse Albedo
   F0
   F90
   Roughness
   Anisotropy
   Normal
   Tangent
   SSS MFP
   SSS MFP Scale
   SSS Phase Anisotropy
   Emissive Color
   Second Roughness
   Second Roughness Weight
   Fuzz Roughness
   Fuzz Amount
   Fuzz Color
   "Glint Density (Disabled)"
   "Glint UVs (Disabled)"
  )
  FunctionOutput(None=@MaterialExpressionStrataSlabBSDF_0) {outputName=Output}
 ]
}
