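# MaterialX uniform EDF as a Substrate material function, for emission that radiates uniformly in all directions.
# The color input, defaulting to white, drives only the Emissive Color of a Substrate Slab BSDF, which is the function's output.
# All other slab inputs are unconnected, so the slab keeps its default diffuse and specular response rather than explicitly zeroing it.
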MaterialFunction /InterchangeAssets/Substrate/MX_UniformEDF
graph {
 properties {description=Constructs an EDF emitting light uniformly in all directions., libraryCategoriesText [MaterialX]}
 nodes [
  n1 = FunctionInput() {bUsePreviewValueAsDefault=true, description=Radiant emittance of light leaving the surface., inputName=color, previewValue=<1 1 1 1>}
  MaterialExpressionStrataSlabBSDF_0 = SubstrateSlabBSDF!(
   Diffuse Albedo
   F0
   F90
   Roughness
   Anisotropy
   Normal
   Tangent
   SSS MFP
   SSS MFP Scale
   SSS Phase Anisotropy
   Emissive Color=@n1
   Second Roughness
   Second Roughness Weight
   Fuzz Roughness
   Fuzz Amount
   Fuzz Color
   "Glint Density (Disabled)"
   "Glint UVs (Disabled)"
  )
  FunctionOutput(None=@MaterialExpressionStrataSlabBSDF_0) {outputName=Output}
 ]
}
