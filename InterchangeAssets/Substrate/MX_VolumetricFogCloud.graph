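# MaterialX Substrate library function for Interchange that builds a volume output from vdf and edf inputs using a Substrate Volumetric Fog Cloud BSDF.
# The vdf material attributes are split, with the BaseColor output feeding both Albedo and Extinction.
# The edf input drives EmissiveColor and falls back to its preview value when unconnected.
# The function is exposed to the material library and returns the BSDF as Volume.
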MaterialFunction /InterchangeAssets/Substrate/MX_VolumetricFogCloud
graph {
 properties {bExposeToLibrary=true, description=A constructor node for the surfaceshader type., libraryCategoriesText [MaterialX Substrate]}
 nodes [
  n1 = GetMaterialAttributes(None=FunctionInput() {inputName=vdf, inputType=FunctionInput_MaterialAttributes}) {attributeGetTypes [#1, #2]}
  FunctionOutput(None=SubstrateVolumetricFogCloudBSDF!(
    Albedo=@n1.BaseColor
    Extinction=@n1.BaseColor
    EmissiveColor=FunctionInput() {bUsePreviewValueAsDefault=true, inputName=edf, sortPriority=1}
    AmbientOcclusion
   )) {outputName=Volume, sortPriority=1}
 ]
}
