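# Substrate material function implementing UsdPreviewSurface, exposed to the library with refraction, displacement and occlusion outputs.
# A select between metalness and specular workflow slabs feeds a conductor branch, mixed with an IOR-based dielectric slab using parameter blending, beneath a weighted clearcoat.
# Opacity, threshold and mode logic derives cutout, presence and transmission values for coverage weighting and simple-volume transmittance.
# Refraction uses a Fresnel-blended IOR approximation, with a separate path-tracing input.
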MaterialFunction /Engine/Functions/Substrate/MF_Substrate_UsdPreviewSurface
graph {
 expressions {
  MaterialExpressionConstant3Vector_0 {class=MaterialExpressionConstant3Vector, properties {constant=<rgba 1 1 1 0>}}
  MaterialExpressionConstant3Vector_1 {class=MaterialExpressionConstant3Vector, properties {constant=<rgba 1 1 1 0>}}
  MaterialExpressionConstant3Vector_3 {class=MaterialExpressionConstant3Vector}
  MaterialExpressionConstant3Vector_4 {class=MaterialExpressionConstant3Vector, properties {constant=<rgba 1 1 1 0>}}
  MaterialExpressionConstant_0 {class=MaterialExpressionConstant, properties {r=1}}
  MaterialExpressionConstant_1 {class=MaterialExpressionConstant}
  MaterialExpressionConstant_2 {class=MaterialExpressionConstant, properties {r=1}}
  MaterialExpressionConstant_4 {class=MaterialExpressionConstant, properties {r=1}}
  MaterialExpressionConstant_5 {class=MaterialExpressionConstant}
  MaterialExpressionDivide_0 {
   class=MaterialExpressionDivide
   inputs [
    A {from=""}
    B {from=MaterialExpressionNamedRerouteDeclaration_4}
   ]
  }
  MaterialExpressionFloor_0 {class=MaterialExpressionFloor, inputs [None {from=MaterialExpressionSaturate_0, output=""}], properties {}}
  MaterialExpressionFresnel_0 {
   class=MaterialExpressionFresnel
   inputs [
    ExponentIn {from="", output=""}
    BaseReflectFractionIn {from=MaterialExpressionNamedRerouteUsage_10, output=""}
    Normal {from="", output=""}
   ]
   properties {}
  }
  MaterialExpressionFunctionInput_0 {
   class=MaterialExpressionFunctionInput
   inputs [Preview {from="", output=""}]
   properties {
    bUsePreviewValueAsDefault=true
    description="Parameter used as diffuseColor when using the specular workflow, when using metallic workflow this is interpreted as albedo."
    inputName=diffuseColor
    previewValue=<0.18 0.18 0.18 0.18>
   }
  }
  MaterialExpressionFunctionInput_1 {
   class=MaterialExpressionFunctionInput
   inputs [Preview {from="", output=""}]
   properties {bUsePreviewValueAsDefault=true, description=Emissive component., inputName=emissiveColor, sortPriority=1}
  }
  MaterialExpressionFunctionInput_10 {
   class=MaterialExpressionFunctionInput
   inputs [Preview {from="", output=""}]
   properties {
    bUsePreviewValueAsDefault=true
    description="When opacity is 1.0 then the gprim is fully opaque, if it is smaller than 1.0 then the prim is translucent, when it is 0 the gprim is fully transparent. Fully transparent objects will still receive a lighting response when the opacityMode is transparent, however they will not receive any lighting when the opacityMode is set to presence."
    inputName=opacity
    inputType=FunctionInput_Scalar
    sortPriority=7
    previewValue=<1 1 1 1>
   }
  }
  MaterialExpressionFunctionInput_11 {
   class=MaterialExpressionFunctionInput
   inputs [Preview {from="", output=""}]
   properties {
    bUsePreviewValueAsDefault=true
    description=Displacement in the direction of the normal.
    inputName=displacement
    inputType=FunctionInput_Scalar
    sortPriority=12
    previewValue=<0 0 0 0>
   }
  }
  MaterialExpressionFunctionInput_12 {
   class=MaterialExpressionFunctionInput
   inputs [Preview {from="", output=""}]
   properties {
    bUsePreviewValueAsDefault=true
    description="Occlusion signal. This provides extra information about theocclusion of different parts of the mesh that this material is appliedto. Occlusion only makes sense as a surface-varying signal, andpathtracers will likely choose to ignore it.  An occlusion value of 0.0means the surface point is fully occluded by other parts of the surface,and a value of 1.0 means the surface point is completely unoccluded byother parts of the surface."
    inputName=occlusion
    inputType=FunctionInput_Scalar
    sortPriority=13
    previewValue=<1 1 1 0>
   }
  }
  MaterialExpressionFunctionInput_13 {
   class=MaterialExpressionFunctionInput
   inputs [Preview {from="", output=""}]
   properties {
    bUsePreviewValueAsDefault=true
    description="This node can fundamentally operate in two modes : Specular workflow where you provide a texture/value to the specularColor input. Or, Metallic workflow where you provide a texture/value to the metallic input. Depending on the 0 or 1 value of this parameter, the following parameters are conditionally enabled: -useSpecularWorkflow = 1: (Specular workflow) -useSpecularWorkflow = 0: (Metalness workflow)"
    inputName=useSpecularWorkflow
    inputType=FunctionInput_Scalar
    sortPriority=13
   }
  }
  MaterialExpressionFunctionInput_14 {
   class=MaterialExpressionFunctionInput
   inputs [Preview {from="", output=""}]
   properties {
    bUsePreviewValueAsDefault=true
    description="This input dictates how materials with zero opacity are interpreted and takes one of two inputs: transparent or presence. When the value is transparent materials with zero opacity still receive a lighting response as, for example, perfectly clear glass still has a specular response. However, when set to presence these materials will not receive any lighting and thus they always serve as cutouts."
    inputName=opacityMode
    inputType=FunctionInput_Scalar
    sortPriority=8
    previewValue=<0 1 1 1>
   }
  }
  MaterialExpressionFunctionInput_2 {
   class=MaterialExpressionFunctionInput
   inputs [Preview {from="", output=""}]
   properties {
    bUsePreviewValueAsDefault=true
    description=Used only in the specular workflow. Specular color to be used. This is the color at 0 incidence. Edge color is assumed white. Transition between the two colors according to Schlick fresnel approximation.
    inputName=specularColor
    sortPriority=2
   }
  }
  MaterialExpressionFunctionInput_3 {
   class=MaterialExpressionFunctionInput
   inputs [Preview {from="", output=""}]
   properties {
    bUsePreviewValueAsDefault=true
    description="Used only in the metalness workflow.     1 for metallic surfaces and 0 for non-metallic. - If metallic is 1, then both F0 (reflectivity at 0 degree incidence) and edge F90 reflectivity will simply be the Albedo.               - If metallic is 0, then Albedo is ignored in the calculation of F0 and F90; F0 is derived from ior via ( (1-ior)/(1+ior) )^2 and F90 is white.                                In between, we interpolate."
    inputName=metallic
    inputType=FunctionInput_Scalar
    sortPriority=3
   }
  }
  MaterialExpressionFunctionInput_4 {
   class=MaterialExpressionFunctionInput
   inputs [Preview {from="", output=""}]
   properties {
    bUsePreviewValueAsDefault=true
    description="Roughness for the specular lobe. The value ranges from 0 to 1, which goes from a perfectly specular surface at 0.0 to maximum roughness of the specular lobe. This value is usually squared before use with a GGX or Beckmann lobe."
    inputName=roughness
    inputType=FunctionInput_Scalar
    sortPriority=4
    previewValue=<0.5 0.5 0.5 0.5>
   }
  }
  MaterialExpressionFunctionInput_5 {
   class=MaterialExpressionFunctionInput
   inputs [Preview {from="", output=""}]
   properties {
    bUsePreviewValueAsDefault=true
    description=Second specular lobe amount. The color is white.
    inputName=clearcoat
    inputType=FunctionInput_Scalar
    sortPriority=5
   }
  }
  MaterialExpressionFunctionInput_6 {
   class=MaterialExpressionFunctionInput
   inputs [Preview {from="", output=""}]
   properties {
    bUsePreviewValueAsDefault=true
    description=Roughness for the second specular lobe.
    inputName=clearcoatRoughness
    inputType=FunctionInput_Scalar
    sortPriority=6
    previewValue=<0.01 0.01 0.01 0.01>
   }
  }
  MaterialExpressionFunctionInput_7 {
   class=MaterialExpressionFunctionInput
   inputs [Preview {from="", output=""}]
   properties {
    bUsePreviewValueAsDefault=true
    description="Index of Refraction to be used for translucent objects and objects with specular components, including the clearcoat if clearcoat > 0."
    inputName=ior
    inputType=FunctionInput_Scalar
    sortPriority=10
    previewValue=<1.5 1.5 1.5 1.5>
   }
  }
  MaterialExpressionFunctionInput_8 {
   class=MaterialExpressionFunctionInput
   inputs [Preview {from="", output=""}]
   properties {
    bUsePreviewValueAsDefault=true
    description="Expects normal in tangent space [(-1,-1,-1), (1,1,1)]This means your texture reader implementation should provide data to this node that is properly scaled and readyto be consumed as a tangent space normal."
    inputName=normal
    sortPriority=11
    previewValue=<0 0 1 1>
   }
  }
  MaterialExpressionFunctionInput_9 {
   class=MaterialExpressionFunctionInput
   inputs [Preview {from="", output=""}]
   properties {
    bUsePreviewValueAsDefault=true
    description="The opacityThreshold input is useful for creating geometric cut-outs based on the opacity input. A value of 0.0 indicates that no masking is applied to the opacity input, while a value greater than 0.0 indicates that rendering of the surface is limited to the areas where the opacity is greater or equal to that value. A classic use of opacityThreshold is to create a leaf from an opacity input texture, in that case the threshold determines the parts of the opacity texture that will be fully transparent and not receive lighting. Note that when opacityThreshold is greater than zero, the opacity values less than the opacityThreshold will not be rendered, and the opacity values greater than or equal to the opacityThreshold will be fully visible. Thus, the opacityThreshold serves as a switch for how the opacity input is interpreted; this “translucent or masked” behavior is common in engines and renderers, and makes the UsdPreviewSurface easier to interchange. Recreating a translucent material that also provides a zero opacity-based mask, can be achieved by setting opacityThreshold to zero and opacityMode to presence."
    inputName=opacityThreshold
    inputType=FunctionInput_Scalar
    sortPriority=9
    previewValue=<0 0 0 0>
   }
  }
  MaterialExpressionFunctionOutput_0 {
   class=MaterialExpressionFunctionOutput
   inputs [None {from=MaterialExpressionSubstrateWeight_1, output=""}]
   properties {outputName=Substrate UsdPreviewSurface, sortPriority=3}
  }
  MaterialExpressionFunctionOutput_1 {
   class=MaterialExpressionFunctionOutput
   inputs [None {from=MaterialExpressionPathTracingQualitySwitch_0, output=""}]
   properties {outputName="Refraction (IOR)", sortPriority=1}
  }
  MaterialExpressionFunctionOutput_2 {
   class=MaterialExpressionFunctionOutput
   inputs [None {from=MaterialExpressionFunctionInput_11, output=""}]
   properties {outputName=Displacement, sortPriority=2}
  }
  MaterialExpressionFunctionOutput_3 {
   class=MaterialExpressionFunctionOutput
   inputs [None {from=MaterialExpressionFunctionInput_12, output=""}]
   properties {outputName=Occlusion}
  }
  MaterialExpressionIf_0 {
   class=MaterialExpressionIf
   inputs [
    A {from=MaterialExpressionNamedRerouteDeclaration_10, output=""}
    B {from=MaterialExpressionFunctionInput_9, output=""}
    "A > B" {from=MaterialExpressionConstant_0, output=""}
    "A == B" {from=MaterialExpressionConstant_0, output=""}
    "A < B" {from=MaterialExpressionConstant_1, output=""}
   ]
   properties {}
  }
  MaterialExpressionIf_1 {
   class=MaterialExpressionIf
   inputs [
    A {from=MaterialExpressionFunctionInput_9, output=""}
    B {from="", output=""}
    "A > B" {from=MaterialExpressionConstant_2, output=""}
    "A == B" {from=MaterialExpressionNamedRerouteUsage_24, output=""}
    "A < B" {from=MaterialExpressionNamedRerouteUsage_24, output=""}
   ]
   properties {}
  }
  MaterialExpressionIf_2 {
   class=MaterialExpressionIf
   inputs [
    A {from=MaterialExpressionNamedRerouteUsage_28, output=""}
    B {from="", output=""}
    "A > B" {from=MaterialExpressionNamedRerouteDeclaration_13, output=""}
    "A == B" {from=MaterialExpressionConstant_5, output=""}
    "A < B" {from=MaterialExpressionNamedRerouteDeclaration_13, output=""}
   ]
   properties {}
  }
  MaterialExpressionIf_3 {
   class=MaterialExpressionIf
   inputs [
    A {from=MaterialExpressionFunctionInput_14, output=""}
    B {from="", output=""}
    "A > B" {from=MaterialExpressionNamedRerouteUsage_30, output=""}
    "A == B" {from=MaterialExpressionNamedRerouteUsage_29, output=""}
    "A < B" {from=MaterialExpressionNamedRerouteUsage_30, output=""}
   ]
   properties {}
  }
  MaterialExpressionLinearInterpolate_0 {
   class=MaterialExpressionLinearInterpolate
   inputs [
    A {from=""}
    B {from=MaterialExpressionDivide_0}
    Alpha {from=MaterialExpressionFresnel_0}
   ]
   properties {constA=1}
  }
  MaterialExpressionLinearInterpolate_1 {
   class=MaterialExpressionLinearInterpolate
   inputs [
    A {from=MaterialExpressionNamedRerouteUsage_21}
    B {from=MaterialExpressionConstant_4}
    Alpha {from=MaterialExpressionFloor_0}
   ]
  }
  MaterialExpressionLinearInterpolate_2 {
   class=MaterialExpressionLinearInterpolate
   inputs [
    A {from=""}
    B {from=MaterialExpressionNamedRerouteUsage_33}
    Alpha {from=MaterialExpressionNamedRerouteUsage_44}
   ]
   properties {constA=0.04, desc=Avoid a fully opaque when F0 is white}
  }
  MaterialExpressionMaterialFunctionCall_0 {
   class=MaterialExpressionMaterialFunctionCall
   inputs [IOR {from=MaterialExpressionNamedRerouteUsage_3}]
   properties {materialFunction=&/Substrate-IOR-To-F0}
  }
  MaterialExpressionNamedRerouteDeclaration_0 {
   class=MaterialExpressionNamedRerouteDeclaration
   inputs [None {from=MaterialExpressionFunctionInput_3}]
   properties {name=Metal, nodeColor=<rgba 1 0.3516 0 1>}
  }
  MaterialExpressionNamedRerouteDeclaration_1 {
   class=MaterialExpressionNamedRerouteDeclaration
   inputs [None {from=MaterialExpressionFunctionInput_0}]
   properties {name=DiffuseColor, nodeColor=<rgba 0.0312 0 1 1>}
  }
  MaterialExpressionNamedRerouteDeclaration_10 {
   class=MaterialExpressionNamedRerouteDeclaration
   inputs [None {from=MaterialExpressionFunctionInput_10}]
   properties {name=Opacity, nodeColor=<rgba 0 1 0.6016 1>}
  }
  MaterialExpressionNamedRerouteDeclaration_12 {
   class=MaterialExpressionNamedRerouteDeclaration
   inputs [None {from=MaterialExpressionLinearInterpolate_1}]
   properties {name=ConductorWeight, nodeColor=<rgba 1 0 0.9844 1>}
  }
  MaterialExpressionNamedRerouteDeclaration_13 {
   class=MaterialExpressionNamedRerouteDeclaration
   inputs [None {from=MaterialExpressionIf_0}]
   properties {name=OpacityCutout, nodeColor=<rgba 0.3125 0 1 1>}
  }
  MaterialExpressionNamedRerouteDeclaration_14 {
   class=MaterialExpressionNamedRerouteDeclaration
   inputs [None {from=MaterialExpressionIf_1}]
   properties {name=TransmissionMix, nodeColor=<rgba 0 1 0.5078 1>}
  }
  MaterialExpressionNamedRerouteDeclaration_15 {
   class=MaterialExpressionNamedRerouteDeclaration
   inputs [None {from=MaterialExpressionFunctionInput_14}]
   properties {name=OpacityMode, nodeColor=<rgba 0.6172 0 1 1>}
  }
  MaterialExpressionNamedRerouteDeclaration_16 {
   class=MaterialExpressionNamedRerouteDeclaration
   inputs [None {from=MaterialExpressionIf_2}]
   properties {name=OpacityPresence, nodeColor=<rgba 1 0 0.6094 1>}
  }
  MaterialExpressionNamedRerouteDeclaration_17 {
   class=MaterialExpressionNamedRerouteDeclaration
   inputs [None {from=MaterialExpressionSwitch_0}]
   properties {name=OpacitySwitch, nodeColor=<rgba 0 0.9297 1 1>}
  }
  MaterialExpressionNamedRerouteDeclaration_2 {
   class=MaterialExpressionNamedRerouteDeclaration
   inputs [None {from=MaterialExpressionFunctionInput_2}]
   properties {name=SpecularColor, nodeColor=<rgba 0.2891 1 0 1>}
  }
  MaterialExpressionNamedRerouteDeclaration_3 {
   class=MaterialExpressionNamedRerouteDeclaration
   inputs [None {from=MaterialExpressionFunctionInput_13}]
   properties {name=SpecularWorkflow, nodeColor=<rgba 1 0 0.6094 1>}
  }
  MaterialExpressionNamedRerouteDeclaration_4 {
   class=MaterialExpressionNamedRerouteDeclaration
   inputs [None {from=MaterialExpressionFunctionInput_7}]
   properties {name=IOR, nodeColor=<rgba 0 0.9297 1 1>}
  }
  MaterialExpressionNamedRerouteDeclaration_5 {
   class=MaterialExpressionNamedRerouteDeclaration
   inputs [None {from=MaterialExpressionFunctionInput_4}]
   properties {name=Roughness, nodeColor=<rgba 1 0.75 0 1>}
  }
  MaterialExpressionNamedRerouteDeclaration_6 {
   class=MaterialExpressionNamedRerouteDeclaration
   inputs [None {from=MaterialExpressionFunctionInput_8}]
   properties {name=Normal, nodeColor=<rgba 0.4297 0 1 1>}
  }
  MaterialExpressionNamedRerouteDeclaration_7 {
   class=MaterialExpressionNamedRerouteDeclaration
   inputs [None {from=MaterialExpressionFunctionInput_1}]
   properties {name=EmissiveColor, nodeColor=<rgba 0 1 0.1094 1>}
  }
  MaterialExpressionNamedRerouteDeclaration_8 {
   class=MaterialExpressionNamedRerouteDeclaration
   inputs [None {from=MaterialExpressionSubstrateSelect_0}]
   properties {name=ConductorSlab, nodeColor=<rgba 1 0 0.7734 1>}
  }
  MaterialExpressionNamedRerouteDeclaration_9 {
   class=MaterialExpressionNamedRerouteDeclaration
   inputs [None {from=MaterialExpressionMaterialFunctionCall_0, output=F0}]
   properties {name=IORtoF0, nodeColor=<rgba 0.7344 0 1 1>}
  }
  MaterialExpressionNamedRerouteUsage_0 {class=MaterialExpressionNamedRerouteUsage}
  MaterialExpressionNamedRerouteUsage_1 {class=MaterialExpressionNamedRerouteUsage}
  MaterialExpressionNamedRerouteUsage_10 {class=MaterialExpressionNamedRerouteUsage}
  MaterialExpressionNamedRerouteUsage_14 {class=MaterialExpressionNamedRerouteUsage}
  MaterialExpressionNamedRerouteUsage_15 {class=MaterialExpressionNamedRerouteUsage}
  MaterialExpressionNamedRerouteUsage_17 {class=MaterialExpressionNamedRerouteUsage}
  MaterialExpressionNamedRerouteUsage_19 {class=MaterialExpressionNamedRerouteUsage}
  MaterialExpressionNamedRerouteUsage_2 {class=MaterialExpressionNamedRerouteUsage}
  MaterialExpressionNamedRerouteUsage_21 {class=MaterialExpressionNamedRerouteUsage}
  MaterialExpressionNamedRerouteUsage_23 {class=MaterialExpressionNamedRerouteUsage}
  MaterialExpressionNamedRerouteUsage_24 {class=MaterialExpressionNamedRerouteUsage}
  MaterialExpressionNamedRerouteUsage_28 {class=MaterialExpressionNamedRerouteUsage}
  MaterialExpressionNamedRerouteUsage_29 {class=MaterialExpressionNamedRerouteUsage}
  MaterialExpressionNamedRerouteUsage_3 {class=MaterialExpressionNamedRerouteUsage}
  MaterialExpressionNamedRerouteUsage_30 {class=MaterialExpressionNamedRerouteUsage}
  MaterialExpressionNamedRerouteUsage_31 {class=MaterialExpressionNamedRerouteUsage}
  MaterialExpressionNamedRerouteUsage_32 {class=MaterialExpressionNamedRerouteUsage}
  MaterialExpressionNamedRerouteUsage_33 {class=MaterialExpressionNamedRerouteUsage}
  MaterialExpressionNamedRerouteUsage_34 {class=MaterialExpressionNamedRerouteUsage}
  MaterialExpressionNamedRerouteUsage_35 {class=MaterialExpressionNamedRerouteUsage}
  MaterialExpressionNamedRerouteUsage_36 {class=MaterialExpressionNamedRerouteUsage}
  MaterialExpressionNamedRerouteUsage_37 {class=MaterialExpressionNamedRerouteUsage}
  MaterialExpressionNamedRerouteUsage_38 {class=MaterialExpressionNamedRerouteUsage}
  MaterialExpressionNamedRerouteUsage_39 {class=MaterialExpressionNamedRerouteUsage}
  MaterialExpressionNamedRerouteUsage_4 {class=MaterialExpressionNamedRerouteUsage}
  MaterialExpressionNamedRerouteUsage_42 {class=MaterialExpressionNamedRerouteUsage}
  MaterialExpressionNamedRerouteUsage_44 {class=MaterialExpressionNamedRerouteUsage}
  MaterialExpressionNamedRerouteUsage_5 {class=MaterialExpressionNamedRerouteUsage}
  MaterialExpressionNamedRerouteUsage_6 {class=MaterialExpressionNamedRerouteUsage}
  MaterialExpressionNamedRerouteUsage_7 {class=MaterialExpressionNamedRerouteUsage}
  MaterialExpressionNamedRerouteUsage_8 {class=MaterialExpressionNamedRerouteUsage}
  MaterialExpressionNamedRerouteUsage_9 {class=MaterialExpressionNamedRerouteUsage}
  MaterialExpressionOneMinus_0 {class=MaterialExpressionOneMinus, inputs [None {from=MaterialExpressionNamedRerouteUsage_32}]}
  MaterialExpressionOneMinus_1 {class=MaterialExpressionOneMinus, inputs [None {from=MaterialExpressionNamedRerouteUsage_42}]}
  MaterialExpressionOneMinus_4 {class=MaterialExpressionOneMinus, inputs [None {from=MaterialExpressionNamedRerouteUsage_38}]}
  MaterialExpressionPathTracingQualitySwitch_0 {
   class=MaterialExpressionPathTracingQualitySwitch
   inputs [
    Normal {from=MaterialExpressionLinearInterpolate_0, output=""}
    PathTraced {from=MaterialExpressionNamedRerouteUsage_2, output=""}
   ]
   properties {}
  }
  MaterialExpressionReroute_0 {class=MaterialExpressionReroute, inputs [None {from=MaterialExpressionNamedRerouteUsage_0}]}
  MaterialExpressionSaturate_0 {class=MaterialExpressionSaturate, inputs [None {from=MaterialExpressionNamedRerouteUsage_14}]}
  MaterialExpressionSubstrateHorizontalMixing_1 {
   class=MaterialExpressionSubstrateHorizontalMixing
   inputs [
    Background {from=MaterialExpressionSubstrateSlabBSDF_2, output=""}
    Foreground {from=MaterialExpressionNamedRerouteUsage_17, output=""}
    Mix {from=MaterialExpressionNamedRerouteUsage_23, output=""}
   ]
   properties {bUseParameterBlending=true}
  }
  MaterialExpressionSubstrateMetalnessToDiffuseAlbedoF0_0 {
   class=MaterialExpressionSubstrateMetalnessToDiffuseAlbedoF0
   inputs [
    BaseColor {from=MaterialExpressionNamedRerouteUsage_0, output=""}
    Metallic {from=MaterialExpressionNamedRerouteUsage_1, output=""}
    Specular {from="", output=""}
   ]
   properties {}
  }
  MaterialExpressionSubstrateSelect_0 {
   class=MaterialExpressionSubstrateSelect
   inputs [
    A {from=MaterialExpressionSubstrateSlabBSDF_1, output=""}
    B {from=MaterialExpressionSubstrateSlabBSDF_5, output=""}
    SelectValue {from=MaterialExpressionFloor_0, output=""}
   ]
   properties {}
  }
  MaterialExpressionSubstrateSlabBSDF_0 {
   class=MaterialExpressionSubstrateSlabBSDF
   inputs [
    Diffuse Albedo {from=MaterialExpressionConstant3Vector_3, output=""}
    F0 {from=MaterialExpressionNamedRerouteUsage_9, output=""}
    F90 {from=MaterialExpressionConstant3Vector_0, output=""}
    Roughness {from=MaterialExpressionFunctionInput_6, output=""}
    Anisotropy {from="", output=""}
    Normal {from=MaterialExpressionNamedRerouteUsage_19, output=""}
    Tangent {from="", output=""}
    SSS MFP {from=MaterialExpressionSubstrateTransmittanceToMFP_0, output=MFP}
    SSS MFP Scale {from="", output=""}
    SSS Phase Anisotropy {from="", output=""}
    Emissive Color {from="", output=""}
    Second Roughness {from="", output=""}
    Second Roughness Weight {from="", output=""}
    Fuzz Roughness {from="", output=""}
    Fuzz Amount {from="", output=""}
    Fuzz Color {from="", output=""}
    "Glint Density (Disabled)" {from="", output=""}
    "Glint UVs (Disabled)" {from="", output=""}
   ]
   properties {}
  }
  MaterialExpressionSubstrateSlabBSDF_1 {
   class=MaterialExpressionSubstrateSlabBSDF
   inputs [
    Diffuse Albedo {from=MaterialExpressionSubstrateMetalnessToDiffuseAlbedoF0_0, output=DiffuseAlbedo}
    F0 {from=MaterialExpressionSubstrateMetalnessToDiffuseAlbedoF0_0, output=DiffuseAlbedo}
    F90 {from=MaterialExpressionReroute_0, output=""}
    Roughness {from=MaterialExpressionNamedRerouteUsage_4, output=""}
    Anisotropy {from="", output=""}
    Normal {from=MaterialExpressionNamedRerouteUsage_5, output=""}
    Tangent {from="", output=""}
    SSS MFP {from=MaterialExpressionSubstrateTransmittanceToMFP_1, output=MFP}
    SSS MFP Scale {from="", output=""}
    SSS Phase Anisotropy {from="", output=""}
    Emissive Color {from=MaterialExpressionNamedRerouteUsage_15, output=""}
    Second Roughness {from="", output=""}
    Second Roughness Weight {from="", output=""}
    Fuzz Roughness {from="", output=""}
    Fuzz Amount {from="", output=""}
    Fuzz Color {from="", output=""}
    "Glint Density (Disabled)" {from="", output=""}
    "Glint UVs (Disabled)" {from="", output=""}
   ]
   properties {subSurfaceType=MSS_SimpleVolume}
  }
  MaterialExpressionSubstrateSlabBSDF_2 {
   class=MaterialExpressionSubstrateSlabBSDF
   inputs [
    Diffuse Albedo {from=MaterialExpressionNamedRerouteUsage_8, output=""}
    F0 {from=MaterialExpressionNamedRerouteDeclaration_9, output=""}
    F90 {from=MaterialExpressionConstant3Vector_1, output=""}
    Roughness {from=MaterialExpressionNamedRerouteUsage_6, output=""}
    Anisotropy {from="", output=""}
    Normal {from=MaterialExpressionNamedRerouteUsage_7, output=""}
    Tangent {from="", output=""}
    SSS MFP {from=MaterialExpressionSubstrateTransmittanceToMFP_2, output=MFP}
    SSS MFP Scale {from="", output=""}
    SSS Phase Anisotropy {from="", output=""}
    Emissive Color {from=MaterialExpressionNamedRerouteUsage_39, output=""}
    Second Roughness {from="", output=""}
    Second Roughness Weight {from="", output=""}
    Fuzz Roughness {from="", output=""}
    Fuzz Amount {from="", output=""}
    Fuzz Color {from="", output=""}
    "Glint Density (Disabled)" {from="", output=""}
    "Glint UVs (Disabled)" {from="", output=""}
   ]
   properties {subSurfaceType=MSS_SimpleVolume}
  }
  MaterialExpressionSubstrateSlabBSDF_5 {
   class=MaterialExpressionSubstrateSlabBSDF
   inputs [
    Diffuse Albedo {from=MaterialExpressionNamedRerouteUsage_37, output=""}
    F0 {from=MaterialExpressionLinearInterpolate_2, output=""}
    F90 {from="", output=""}
    Roughness {from=MaterialExpressionNamedRerouteUsage_34, output=""}
    Anisotropy {from="", output=""}
    Normal {from=MaterialExpressionNamedRerouteUsage_35, output=""}
    Tangent {from="", output=""}
    SSS MFP {from=MaterialExpressionSubstrateTransmittanceToMFP_5, output=MFP}
    SSS MFP Scale {from="", output=""}
    SSS Phase Anisotropy {from="", output=""}
    Emissive Color {from=MaterialExpressionNamedRerouteUsage_36, output=""}
    Second Roughness {from="", output=""}
    Second Roughness Weight {from="", output=""}
    Fuzz Roughness {from="", output=""}
    Fuzz Amount {from="", output=""}
    Fuzz Color {from="", output=""}
    "Glint Density (Disabled)" {from="", output=""}
    "Glint UVs (Disabled)" {from="", output=""}
   ]
   properties {desc=Simple Volume slab used for transparent mode, subSurfaceType=MSS_SimpleVolume}
  }
  MaterialExpressionSubstrateTransmittanceToMFP_0 {
   class=MaterialExpressionSubstrateTransmittanceToMFP
   inputs [
    TransmittanceColor {from=MaterialExpressionConstant3Vector_0, output=""}
    Thickness {from="", output=""}
   ]
   properties {}
  }
  MaterialExpressionSubstrateTransmittanceToMFP_1 {
   class=MaterialExpressionSubstrateTransmittanceToMFP
   inputs [
    TransmittanceColor {from=MaterialExpressionOneMinus_0, output=""}
    Thickness {from="", output=""}
   ]
   properties {}
  }
  MaterialExpressionSubstrateTransmittanceToMFP_2 {
   class=MaterialExpressionSubstrateTransmittanceToMFP
   inputs [
    TransmittanceColor {from=MaterialExpressionOneMinus_1, output=""}
    Thickness {from="", output=""}
   ]
   properties {}
  }
  MaterialExpressionSubstrateTransmittanceToMFP_5 {
   class=MaterialExpressionSubstrateTransmittanceToMFP
   inputs [
    TransmittanceColor {from=MaterialExpressionOneMinus_4, output=""}
    Thickness {from="", output=""}
   ]
   properties {}
  }
  MaterialExpressionSubstrateVerticalLayering_1 {
   class=MaterialExpressionSubstrateVerticalLayering
   inputs [
    Top {from=MaterialExpressionSubstrateWeight_0, output=""}
    Bottom {from=MaterialExpressionSubstrateHorizontalMixing_1, output=""}
    Top Thickness {from="", output=""}
   ]
   properties {}
  }
  MaterialExpressionSubstrateWeight_0 {
   class=MaterialExpressionSubstrateWeight
   inputs [
    A {from=MaterialExpressionSubstrateSlabBSDF_0, output=""}
    Weight {from=MaterialExpressionFunctionInput_5, output=""}
   ]
   properties {}
  }
  MaterialExpressionSubstrateWeight_1 {
   class=MaterialExpressionSubstrateWeight
   inputs [
    A {from=MaterialExpressionSubstrateVerticalLayering_1, output=""}
    Weight {from=MaterialExpressionNamedRerouteUsage_31, output=""}
   ]
   properties {}
  }
  MaterialExpressionSwitch_0 {
   class=MaterialExpressionSwitch
   inputs [
    SwitchValue {from=MaterialExpressionNamedRerouteDeclaration_15, output=""}
    Default {from=MaterialExpressionNamedRerouteUsage_29, output=""}
    None {from=MaterialExpressionNamedRerouteUsage_29, output=""}
    None {from=MaterialExpressionNamedRerouteUsage_30, output=""}
   ]
   properties {inputs [
     {
      inputName=None
      input {
       expression=&self:MaterialExpressionNamedRerouteUsage_29
       inputName=None
       mask=0
       maskA=0
       maskB=0
       maskG=0
       maskR=0
       outputIndex=0
      }
     }
     {
      inputName=None
      input {
       expression=&self:MaterialExpressionNamedRerouteUsage_30
       inputName=None
       mask=0
       maskA=0
       maskB=0
       maskG=0
       maskR=0
       outputIndex=0
      }
     }
    ]}
  }
 }
 properties {bExposeToLibrary=true, libraryCategoriesText [MaterialX Substrate]}
}
