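# Substrate material implementing USD Preview Surface, feeding parameter-driven channels into MF_Substrate_UsdPreviewSurface.
# Channels such as roughness and opacity lerp between constants and textures via scalar weights, using ChannelSelector for single-value inputs; diffuse can also blend in linearized vertex color.
# Each texture has a UV index and UsdTransform2d transform.
# UVSwitch composites pick among four UV lookups with chained Ifs, replacing Custom nodes whose analytic derivatives fail under Nanite.
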Material /InterchangeAssets/Substrate/UsdPreviewSurface
graph {
 expressions {
  MaterialExpressionComposite_0 {
   class=MaterialExpressionComposite
   inputs [UVIndex {from=MaterialExpressionScalarParameter_33, output=""}]
   properties {
    desc="Used as a replacement for Custom expression, since it doesn't fully translates to analytical derivatives for Nanite"
    subgraphName=UVSwitch
   }
  }
  MaterialExpressionComposite_1 {
   class=MaterialExpressionComposite
   inputs [UVIndex {from=MaterialExpressionScalarParameter_36, output=""}]
   properties {
    desc="Used as a replacement for Custom expression, since it doesn't fully translates to analytical derivatives for Nanite"
    subgraphName=UVSwitch
   }
  }
  MaterialExpressionComposite_12 {
   class=MaterialExpressionComposite
   inputs [UVIndex {from=MaterialExpressionScalarParameter_16, output=""}]
   properties {
    desc="Used as a replacement for Custom expression, since it doesn't fully translates to analytical derivatives for Nanite"
    subgraphName=UVSwitch
   }
  }
  MaterialExpressionComposite_13 {
   class=MaterialExpressionComposite
   inputs [UVIndex {from=MaterialExpressionScalarParameter_17, output=""}]
   properties {
    desc="Used as a replacement for Custom expression, since it doesn't fully translates to analytical derivatives for Nanite"
    subgraphName=UVSwitch
   }
  }
  MaterialExpressionComposite_14 {
   class=MaterialExpressionComposite
   inputs [UVIndex {from=MaterialExpressionScalarParameter_18, output=""}]
   properties {
    desc="Used as a replacement for Custom expression, since it doesn't fully translates to analytical derivatives for Nanite"
    subgraphName=UVSwitch
   }
  }
  MaterialExpressionComposite_15 {
   class=MaterialExpressionComposite
   inputs [UVIndex {from=MaterialExpressionScalarParameter_19, output=""}]
   properties {
    desc="Used as a replacement for Custom expression, since it doesn't fully translates to analytical derivatives for Nanite"
    subgraphName=UVSwitch
   }
  }
  MaterialExpressionComposite_16 {
   class=MaterialExpressionComposite
   inputs [UVIndex {from=MaterialExpressionScalarParameter_20, output=""}]
   properties {
    desc="Used as a replacement for Custom expression, since it doesn't fully translates to analytical derivatives for Nanite"
    subgraphName=UVSwitch
   }
  }
  MaterialExpressionComposite_17 {
   class=MaterialExpressionComposite
   inputs [UVIndex {from=MaterialExpressionScalarParameter_21, output=""}]
   properties {
    desc="Used as a replacement for Custom expression, since it doesn't fully translates to analytical derivatives for Nanite"
    subgraphName=UVSwitch
   }
  }
  MaterialExpressionComposite_18 {
   class=MaterialExpressionComposite
   inputs [UVIndex {from=MaterialExpressionScalarParameter_22, output=""}]
   properties {
    desc="Used as a replacement for Custom expression, since it doesn't fully translates to analytical derivatives for Nanite"
    subgraphName=UVSwitch
   }
  }
  MaterialExpressionComposite_2 {
   class=MaterialExpressionComposite
   inputs [UVIndex {from=MaterialExpressionScalarParameter_40, output=""}]
   properties {
    desc="Used as a replacement for Custom expression, since it doesn't fully translates to analytical derivatives for Nanite"
    subgraphName=UVSwitch
   }
  }
  MaterialExpressionComposite_3 {
   class=MaterialExpressionComposite
   inputs [UVIndex {from=MaterialExpressionScalarParameter_45, output=""}]
   properties {
    desc="Used as a replacement for Custom expression, since it doesn't fully translates to analytical derivatives for Nanite"
    subgraphName=UVSwitch
   }
  }
  MaterialExpressionComposite_4 {
   class=MaterialExpressionComposite
   inputs [UVIndex {from=MaterialExpressionScalarParameter_50, output=""}]
   properties {
    desc="Used as a replacement for Custom expression, since it doesn't fully translates to analytical derivatives for Nanite"
    subgraphName=UVSwitch
   }
  }
  MaterialExpressionComposite_7 {
   class=MaterialExpressionComposite
   inputs [UVIndex {from=MaterialExpressionScalarParameter_15, output=""}]
   properties {
    desc="Used as a replacement for Custom expression, since it doesn't fully translates to analytical derivatives for Nanite"
    subgraphName=UVSwitch
   }
  }
  MaterialExpressionConvert_0 {
   class=MaterialExpressionConvert
   inputs [None {from=MaterialExpressionVectorParameter_17, output=RGBA}]
   properties {
    nodeName=Make Vector2
    convertInputs [{
      type=Vector4
      defaultValue=<rgba 0 0 0 1>
      expressionInput {
       expression=&self:MaterialExpressionVectorParameter_17
       inputName=None
       mask=1
       maskA=1
       maskB=1
       maskG=1
       maskR=1
       outputIndex=5
      }
     }]
    convertOutputs [{type=Vector2}, {type=Vector2}]
   }
  }
  MaterialExpressionConvert_1 {
   class=MaterialExpressionConvert
   inputs [None {from=MaterialExpressionVectorParameter_11, output=RGBA}]
   properties {
    nodeName=Make Vector2
    convertInputs [{
      type=Vector4
      defaultValue=<rgba 0 0 0 1>
      expressionInput {
       expression=&self:MaterialExpressionVectorParameter_11
       inputName=None
       mask=1
       maskA=1
       maskB=1
       maskG=1
       maskR=1
       outputIndex=5
      }
     }]
    convertOutputs [{type=Vector2}, {type=Vector2}]
   }
  }
  MaterialExpressionConvert_10 {
   class=MaterialExpressionConvert
   inputs [None {from=MaterialExpressionVectorParameter_21, output=RGBA}]
   properties {
    nodeName=Make Vector2
    convertInputs [{
      type=Vector4
      defaultValue=<rgba 0 0 0 1>
      expressionInput {
       expression=&self:MaterialExpressionVectorParameter_21
       inputName=None
       mask=1
       maskA=1
       maskB=1
       maskG=1
       maskR=1
       outputIndex=5
      }
     }]
    convertOutputs [{type=Vector2}, {type=Vector2}]
   }
  }
  MaterialExpressionConvert_11 {
   class=MaterialExpressionConvert
   inputs [None {from=MaterialExpressionVectorParameter_23, output=RGBA}]
   properties {
    nodeName=Make Vector2
    convertInputs [{
      type=Vector4
      defaultValue=<rgba 0 0 0 1>
      expressionInput {
       expression=&self:MaterialExpressionVectorParameter_23
       inputName=None
       mask=1
       maskA=1
       maskB=1
       maskG=1
       maskR=1
       outputIndex=5
      }
     }]
    convertOutputs [{type=Vector2}, {type=Vector2}]
   }
  }
  MaterialExpressionConvert_12 {
   class=MaterialExpressionConvert
   inputs [None {from=MaterialExpressionVectorParameter_25, output=RGBA}]
   properties {
    nodeName=Make Vector2
    convertInputs [{
      type=Vector4
      defaultValue=<rgba 0 0 0 1>
      expressionInput {
       expression=&self:MaterialExpressionVectorParameter_25
       inputName=None
       mask=1
       maskA=1
       maskB=1
       maskG=1
       maskR=1
       outputIndex=5
      }
     }]
    convertOutputs [{type=Vector2}, {type=Vector2}]
   }
  }
  MaterialExpressionConvert_2 {
   class=MaterialExpressionConvert
   inputs [None {from=MaterialExpressionVectorParameter_8, output=RGBA}]
   properties {
    nodeName=Make Vector2
    convertInputs [{
      type=Vector4
      defaultValue=<rgba 0 0 0 1>
      expressionInput {
       expression=&self:MaterialExpressionVectorParameter_8
       inputName=None
       mask=1
       maskA=1
       maskB=1
       maskG=1
       maskR=1
       outputIndex=5
      }
     }]
    convertOutputs [{type=Vector2}, {type=Vector2}]
   }
  }
  MaterialExpressionConvert_3 {
   class=MaterialExpressionConvert
   inputs [None {from=MaterialExpressionVectorParameter_9, output=RGBA}]
   properties {
    nodeName=Make Vector2
    convertInputs [{
      type=Vector4
      defaultValue=<rgba 0 0 0 1>
      expressionInput {
       expression=&self:MaterialExpressionVectorParameter_9
       inputName=None
       mask=1
       maskA=1
       maskB=1
       maskG=1
       maskR=1
       outputIndex=5
      }
     }]
    convertOutputs [{type=Vector2}, {type=Vector2}]
   }
  }
  MaterialExpressionConvert_4 {
   class=MaterialExpressionConvert
   inputs [None {from=MaterialExpressionVectorParameter_10, output=RGBA}]
   properties {
    nodeName=Make Vector2
    convertInputs [{
      type=Vector4
      defaultValue=<rgba 0 0 0 1>
      expressionInput {
       expression=&self:MaterialExpressionVectorParameter_10
       inputName=None
       mask=1
       maskA=1
       maskB=1
       maskG=1
       maskR=1
       outputIndex=5
      }
     }]
    convertOutputs [{type=Vector2}, {type=Vector2}]
   }
  }
  MaterialExpressionConvert_5 {
   class=MaterialExpressionConvert
   inputs [None {from=MaterialExpressionVectorParameter_12, output=RGBA}]
   properties {
    nodeName=Make Vector2
    convertInputs [{
      type=Vector4
      defaultValue=<rgba 0 0 0 1>
      expressionInput {
       expression=&self:MaterialExpressionVectorParameter_12
       inputName=None
       mask=1
       maskA=1
       maskB=1
       maskG=1
       maskR=1
       outputIndex=5
      }
     }]
    convertOutputs [{type=Vector2}, {type=Vector2}]
   }
  }
  MaterialExpressionConvert_6 {
   class=MaterialExpressionConvert
   inputs [None {from=MaterialExpressionVectorParameter_13, output=RGBA}]
   properties {
    nodeName=Make Vector2
    convertInputs [{
      type=Vector4
      defaultValue=<rgba 0 0 0 1>
      expressionInput {
       expression=&self:MaterialExpressionVectorParameter_13
       inputName=None
       mask=1
       maskA=1
       maskB=1
       maskG=1
       maskR=1
       outputIndex=5
      }
     }]
    convertOutputs [{type=Vector2}, {type=Vector2}]
   }
  }
  MaterialExpressionConvert_7 {
   class=MaterialExpressionConvert
   inputs [None {from=MaterialExpressionVectorParameter_14, output=RGBA}]
   properties {
    nodeName=Make Vector2
    convertInputs [{
      type=Vector4
      defaultValue=<rgba 0 0 0 1>
      expressionInput {
       expression=&self:MaterialExpressionVectorParameter_14
       inputName=None
       mask=1
       maskA=1
       maskB=1
       maskG=1
       maskR=1
       outputIndex=5
      }
     }]
    convertOutputs [{type=Vector2}, {type=Vector2}]
   }
  }
  MaterialExpressionConvert_8 {
   class=MaterialExpressionConvert
   inputs [None {from=MaterialExpressionVectorParameter_15, output=RGBA}]
   properties {
    nodeName=Make Vector2
    convertInputs [{
      type=Vector4
      defaultValue=<rgba 0 0 0 1>
      expressionInput {
       expression=&self:MaterialExpressionVectorParameter_15
       inputName=None
       mask=1
       maskA=1
       maskB=1
       maskG=1
       maskR=1
       outputIndex=5
      }
     }]
    convertOutputs [{type=Vector2}, {type=Vector2}]
   }
  }
  MaterialExpressionConvert_9 {
   class=MaterialExpressionConvert
   inputs [None {from=MaterialExpressionVectorParameter_19, output=RGBA}]
   properties {
    nodeName=Make Vector2
    convertInputs [{
      type=Vector4
      defaultValue=<rgba 0 0 0 1>
      expressionInput {
       expression=&self:MaterialExpressionVectorParameter_19
       inputName=None
       mask=1
       maskA=1
       maskB=1
       maskG=1
       maskR=1
       outputIndex=5
      }
     }]
    convertOutputs [{type=Vector2}, {type=Vector2}]
   }
  }
  MaterialExpressionIf_0 {
   class=MaterialExpressionIf
   inputs [
    A {from=MaterialExpressionReroute_16, output=UVIndex}
    B {from="", output=""}
    "A > B" {from=MaterialExpressionIf_1, output=""}
    "A == B" {from=MaterialExpressionIf_1, output=""}
    "A < B" {from=MaterialExpressionTextureCoordinate_0, output=""}
   ]
   properties {constB=1}
  }
  MaterialExpressionIf_1 {
   class=MaterialExpressionIf
   inputs [
    A {from=MaterialExpressionReroute_16, output=UVIndex}
    B {from="", output=""}
    "A > B" {from=MaterialExpressionIf_2, output=""}
    "A == B" {from=MaterialExpressionIf_2, output=""}
    "A < B" {from=MaterialExpressionTextureCoordinate_1, output=""}
   ]
   properties {constB=2}
  }
  MaterialExpressionIf_10 {
   class=MaterialExpressionIf
   inputs [
    A {from=MaterialExpressionReroute_24, output=UVIndex}
    B {from="", output=""}
    "A > B" {from=MaterialExpressionIf_11, output=""}
    "A == B" {from=MaterialExpressionIf_11, output=""}
    "A < B" {from=MaterialExpressionTextureCoordinate_13, output=""}
   ]
   properties {constB=2}
  }
  MaterialExpressionIf_11 {
   class=MaterialExpressionIf
   inputs [
    A {from=MaterialExpressionReroute_24, output=UVIndex}
    B {from="", output=""}
    "A > B" {from=MaterialExpressionTextureCoordinate_15, output=""}
    "A == B" {from=MaterialExpressionTextureCoordinate_15, output=""}
    "A < B" {from=MaterialExpressionTextureCoordinate_14, output=""}
   ]
   properties {constB=3}
  }
  MaterialExpressionIf_12 {
   class=MaterialExpressionIf
   inputs [
    A {from=MaterialExpressionReroute_26, output=UVIndex}
    B {from="", output=""}
    "A > B" {from=MaterialExpressionIf_13, output=""}
    "A == B" {from=MaterialExpressionIf_13, output=""}
    "A < B" {from=MaterialExpressionTextureCoordinate_16, output=""}
   ]
   properties {constB=1}
  }
  MaterialExpressionIf_13 {
   class=MaterialExpressionIf
   inputs [
    A {from=MaterialExpressionReroute_26, output=UVIndex}
    B {from="", output=""}
    "A > B" {from=MaterialExpressionIf_14, output=""}
    "A == B" {from=MaterialExpressionIf_14, output=""}
    "A < B" {from=MaterialExpressionTextureCoordinate_17, output=""}
   ]
   properties {constB=2}
  }
  MaterialExpressionIf_14 {
   class=MaterialExpressionIf
   inputs [
    A {from=MaterialExpressionReroute_26, output=UVIndex}
    B {from="", output=""}
    "A > B" {from=MaterialExpressionTextureCoordinate_19, output=""}
    "A == B" {from=MaterialExpressionTextureCoordinate_19, output=""}
    "A < B" {from=MaterialExpressionTextureCoordinate_18, output=""}
   ]
   properties {constB=3}
  }
  MaterialExpressionIf_2 {
   class=MaterialExpressionIf
   inputs [
    A {from=MaterialExpressionReroute_16, output=UVIndex}
    B {from="", output=""}
    "A > B" {from=MaterialExpressionTextureCoordinate_3, output=""}
    "A == B" {from=MaterialExpressionTextureCoordinate_3, output=""}
    "A < B" {from=MaterialExpressionTextureCoordinate_2, output=""}
   ]
   properties {constB=3}
  }
  MaterialExpressionIf_22 {
   class=MaterialExpressionIf
   inputs [
    A {from=MaterialExpressionReroute_24, output=UVIndex}
    B {from="", output=""}
    "A > B" {from=MaterialExpressionIf_23, output=""}
    "A == B" {from=MaterialExpressionIf_23, output=""}
    "A < B" {from=MaterialExpressionTextureCoordinate_32, output=""}
   ]
   properties {constB=1}
  }
  MaterialExpressionIf_23 {
   class=MaterialExpressionIf
   inputs [
    A {from=MaterialExpressionReroute_24, output=UVIndex}
    B {from="", output=""}
    "A > B" {from=MaterialExpressionIf_24, output=""}
    "A == B" {from=MaterialExpressionIf_24, output=""}
    "A < B" {from=MaterialExpressionTextureCoordinate_33, output=""}
   ]
   properties {constB=2}
  }
  MaterialExpressionIf_24 {
   class=MaterialExpressionIf
   inputs [
    A {from=MaterialExpressionReroute_24, output=UVIndex}
    B {from="", output=""}
    "A > B" {from=MaterialExpressionTextureCoordinate_35, output=""}
    "A == B" {from=MaterialExpressionTextureCoordinate_35, output=""}
    "A < B" {from=MaterialExpressionTextureCoordinate_34, output=""}
   ]
   properties {constB=3}
  }
  MaterialExpressionIf_3 {
   class=MaterialExpressionIf
   inputs [
    A {from=MaterialExpressionReroute_2, output=UVIndex}
    B {from="", output=""}
    "A > B" {from=MaterialExpressionIf_4, output=""}
    "A == B" {from=MaterialExpressionIf_4, output=""}
    "A < B" {from=MaterialExpressionTextureCoordinate_4, output=""}
   ]
   properties {constB=1}
  }
  MaterialExpressionIf_37 {
   class=MaterialExpressionIf
   inputs [
    A {from=MaterialExpressionReroute_0, output=UVIndex}
    B {from="", output=""}
    "A > B" {from=MaterialExpressionIf_38, output=""}
    "A == B" {from=MaterialExpressionIf_38, output=""}
    "A < B" {from=MaterialExpressionTextureCoordinate_52, output=""}
   ]
   properties {constB=1}
  }
  MaterialExpressionIf_38 {
   class=MaterialExpressionIf
   inputs [
    A {from=MaterialExpressionReroute_0, output=UVIndex}
    B {from="", output=""}
    "A > B" {from=MaterialExpressionIf_39, output=""}
    "A == B" {from=MaterialExpressionIf_39, output=""}
    "A < B" {from=MaterialExpressionTextureCoordinate_53, output=""}
   ]
   properties {constB=2}
  }
  MaterialExpressionIf_39 {
   class=MaterialExpressionIf
   inputs [
    A {from=MaterialExpressionReroute_0, output=UVIndex}
    B {from="", output=""}
    "A > B" {from=MaterialExpressionTextureCoordinate_55, output=""}
    "A == B" {from=MaterialExpressionTextureCoordinate_55, output=""}
    "A < B" {from=MaterialExpressionTextureCoordinate_54, output=""}
   ]
   properties {constB=3}
  }
  MaterialExpressionIf_4 {
   class=MaterialExpressionIf
   inputs [
    A {from=MaterialExpressionReroute_2, output=UVIndex}
    B {from="", output=""}
    "A > B" {from=MaterialExpressionIf_5, output=""}
    "A == B" {from=MaterialExpressionIf_5, output=""}
    "A < B" {from=MaterialExpressionTextureCoordinate_5, output=""}
   ]
   properties {constB=2}
  }
  MaterialExpressionIf_40 {
   class=MaterialExpressionIf
   inputs [
    A {from=MaterialExpressionReroute_0, output=UVIndex}
    B {from="", output=""}
    "A > B" {from=MaterialExpressionIf_41, output=""}
    "A == B" {from=MaterialExpressionIf_41, output=""}
    "A < B" {from=MaterialExpressionTextureCoordinate_56, output=""}
   ]
   properties {constB=1}
  }
  MaterialExpressionIf_41 {
   class=MaterialExpressionIf
   inputs [
    A {from=MaterialExpressionReroute_0, output=UVIndex}
    B {from="", output=""}
    "A > B" {from=MaterialExpressionIf_42, output=""}
    "A == B" {from=MaterialExpressionIf_42, output=""}
    "A < B" {from=MaterialExpressionTextureCoordinate_57, output=""}
   ]
   properties {constB=2}
  }
  MaterialExpressionIf_42 {
   class=MaterialExpressionIf
   inputs [
    A {from=MaterialExpressionReroute_0, output=UVIndex}
    B {from="", output=""}
    "A > B" {from=MaterialExpressionTextureCoordinate_59, output=""}
    "A == B" {from=MaterialExpressionTextureCoordinate_59, output=""}
    "A < B" {from=MaterialExpressionTextureCoordinate_58, output=""}
   ]
   properties {constB=3}
  }
  MaterialExpressionIf_43 {
   class=MaterialExpressionIf
   inputs [
    A {from=MaterialExpressionReroute_20, output=UVIndex}
    B {from="", output=""}
    "A > B" {from=MaterialExpressionIf_44, output=""}
    "A == B" {from=MaterialExpressionIf_44, output=""}
    "A < B" {from=MaterialExpressionTextureCoordinate_60, output=""}
   ]
   properties {constB=1}
  }
  MaterialExpressionIf_44 {
   class=MaterialExpressionIf
   inputs [
    A {from=MaterialExpressionReroute_20, output=UVIndex}
    B {from="", output=""}
    "A > B" {from=MaterialExpressionIf_45, output=""}
    "A == B" {from=MaterialExpressionIf_45, output=""}
    "A < B" {from=MaterialExpressionTextureCoordinate_61, output=""}
   ]
   properties {constB=2}
  }
  MaterialExpressionIf_45 {
   class=MaterialExpressionIf
   inputs [
    A {from=MaterialExpressionReroute_20, output=UVIndex}
    B {from="", output=""}
    "A > B" {from=MaterialExpressionTextureCoordinate_63, output=""}
    "A == B" {from=MaterialExpressionTextureCoordinate_63, output=""}
    "A < B" {from=MaterialExpressionTextureCoordinate_62, output=""}
   ]
   properties {constB=3}
  }
  MaterialExpressionIf_46 {
   class=MaterialExpressionIf
   inputs [
    A {from=MaterialExpressionReroute_0, output=UVIndex}
    B {from="", output=""}
    "A > B" {from=MaterialExpressionIf_47, output=""}
    "A == B" {from=MaterialExpressionIf_47, output=""}
    "A < B" {from=MaterialExpressionTextureCoordinate_64, output=""}
   ]
   properties {constB=1}
  }
  MaterialExpressionIf_47 {
   class=MaterialExpressionIf
   inputs [
    A {from=MaterialExpressionReroute_0, output=UVIndex}
    B {from="", output=""}
    "A > B" {from=MaterialExpressionIf_48, output=""}
    "A == B" {from=MaterialExpressionIf_48, output=""}
    "A < B" {from=MaterialExpressionTextureCoordinate_65, output=""}
   ]
   properties {constB=2}
  }
  MaterialExpressionIf_48 {
   class=MaterialExpressionIf
   inputs [
    A {from=MaterialExpressionReroute_0, output=UVIndex}
    B {from="", output=""}
    "A > B" {from=MaterialExpressionTextureCoordinate_67, output=""}
    "A == B" {from=MaterialExpressionTextureCoordinate_67, output=""}
    "A < B" {from=MaterialExpressionTextureCoordinate_66, output=""}
   ]
   properties {constB=3}
  }
  MaterialExpressionIf_49 {
   class=MaterialExpressionIf
   inputs [
    A {from=MaterialExpressionReroute_26, output=UVIndex}
    B {from="", output=""}
    "A > B" {from=MaterialExpressionIf_50, output=""}
    "A == B" {from=MaterialExpressionIf_50, output=""}
    "A < B" {from=MaterialExpressionTextureCoordinate_68, output=""}
   ]
   properties {constB=1}
  }
  MaterialExpressionIf_5 {
   class=MaterialExpressionIf
   inputs [
    A {from=MaterialExpressionReroute_2, output=UVIndex}
    B {from="", output=""}
    "A > B" {from=MaterialExpressionTextureCoordinate_7, output=""}
    "A == B" {from=MaterialExpressionTextureCoordinate_7, output=""}
    "A < B" {from=MaterialExpressionTextureCoordinate_6, output=""}
   ]
   properties {constB=3}
  }
  MaterialExpressionIf_50 {
   class=MaterialExpressionIf
   inputs [
    A {from=MaterialExpressionReroute_26, output=UVIndex}
    B {from="", output=""}
    "A > B" {from=MaterialExpressionIf_51, output=""}
    "A == B" {from=MaterialExpressionIf_51, output=""}
    "A < B" {from=MaterialExpressionTextureCoordinate_69, output=""}
   ]
   properties {constB=2}
  }
  MaterialExpressionIf_51 {
   class=MaterialExpressionIf
   inputs [
    A {from=MaterialExpressionReroute_26, output=UVIndex}
    B {from="", output=""}
    "A > B" {from=MaterialExpressionTextureCoordinate_71, output=""}
    "A == B" {from=MaterialExpressionTextureCoordinate_71, output=""}
    "A < B" {from=MaterialExpressionTextureCoordinate_70, output=""}
   ]
   properties {constB=3}
  }
  MaterialExpressionIf_52 {
   class=MaterialExpressionIf
   inputs [
    A {from=MaterialExpressionReroute_28, output=UVIndex}
    B {from="", output=""}
    "A > B" {from=MaterialExpressionIf_53, output=""}
    "A == B" {from=MaterialExpressionIf_53, output=""}
    "A < B" {from=MaterialExpressionTextureCoordinate_72, output=""}
   ]
   properties {constB=1}
  }
  MaterialExpressionIf_53 {
   class=MaterialExpressionIf
   inputs [
    A {from=MaterialExpressionReroute_28, output=UVIndex}
    B {from="", output=""}
    "A > B" {from=MaterialExpressionIf_54, output=""}
    "A == B" {from=MaterialExpressionIf_54, output=""}
    "A < B" {from=MaterialExpressionTextureCoordinate_73, output=""}
   ]
   properties {constB=2}
  }
  MaterialExpressionIf_54 {
   class=MaterialExpressionIf
   inputs [
    A {from=MaterialExpressionReroute_28, output=UVIndex}
    B {from="", output=""}
    "A > B" {from=MaterialExpressionTextureCoordinate_75, output=""}
    "A == B" {from=MaterialExpressionTextureCoordinate_75, output=""}
    "A < B" {from=MaterialExpressionTextureCoordinate_74, output=""}
   ]
   properties {constB=3}
  }
  MaterialExpressionIf_55 {
   class=MaterialExpressionIf
   inputs [
    A {from=MaterialExpressionReroute_0, output=UVIndex}
    B {from="", output=""}
    "A > B" {from=MaterialExpressionIf_56, output=""}
    "A == B" {from=MaterialExpressionIf_56, output=""}
    "A < B" {from=MaterialExpressionTextureCoordinate_76, output=""}
   ]
   properties {constB=1}
  }
  MaterialExpressionIf_56 {
   class=MaterialExpressionIf
   inputs [
    A {from=MaterialExpressionReroute_0, output=UVIndex}
    B {from="", output=""}
    "A > B" {from=MaterialExpressionIf_57, output=""}
    "A == B" {from=MaterialExpressionIf_57, output=""}
    "A < B" {from=MaterialExpressionTextureCoordinate_77, output=""}
   ]
   properties {constB=2}
  }
  MaterialExpressionIf_57 {
   class=MaterialExpressionIf
   inputs [
    A {from=MaterialExpressionReroute_0, output=UVIndex}
    B {from="", output=""}
    "A > B" {from=MaterialExpressionTextureCoordinate_79, output=""}
    "A == B" {from=MaterialExpressionTextureCoordinate_79, output=""}
    "A < B" {from=MaterialExpressionTextureCoordinate_78, output=""}
   ]
   properties {constB=3}
  }
  MaterialExpressionIf_6 {
   class=MaterialExpressionIf
   inputs [
    A {from=MaterialExpressionReroute_3, output=UVIndex}
    B {from="", output=""}
    "A > B" {from=MaterialExpressionIf_7, output=""}
    "A == B" {from=MaterialExpressionIf_7, output=""}
    "A < B" {from=MaterialExpressionTextureCoordinate_8, output=""}
   ]
   properties {constB=1}
  }
  MaterialExpressionIf_7 {
   class=MaterialExpressionIf
   inputs [
    A {from=MaterialExpressionReroute_3, output=UVIndex}
    B {from="", output=""}
    "A > B" {from=MaterialExpressionIf_8, output=""}
    "A == B" {from=MaterialExpressionIf_8, output=""}
    "A < B" {from=MaterialExpressionTextureCoordinate_9, output=""}
   ]
   properties {constB=2}
  }
  MaterialExpressionIf_8 {
   class=MaterialExpressionIf
   inputs [
    A {from=MaterialExpressionReroute_3, output=UVIndex}
    B {from="", output=""}
    "A > B" {from=MaterialExpressionTextureCoordinate_11, output=""}
    "A == B" {from=MaterialExpressionTextureCoordinate_11, output=""}
    "A < B" {from=MaterialExpressionTextureCoordinate_10, output=""}
   ]
   properties {constB=3}
  }
  MaterialExpressionIf_9 {
   class=MaterialExpressionIf
   inputs [
    A {from=MaterialExpressionReroute_24, output=UVIndex}
    B {from="", output=""}
    "A > B" {from=MaterialExpressionIf_10, output=""}
    "A == B" {from=MaterialExpressionIf_10, output=""}
    "A < B" {from=MaterialExpressionTextureCoordinate_12, output=""}
   ]
   properties {constB=1}
  }
  MaterialExpressionLinearInterpolate_0 {
   class=MaterialExpressionLinearInterpolate
   inputs [
    A {from=MaterialExpressionLinearInterpolate_8}
    B {from=MaterialExpressionTextureSampleParameter2D_0, output=RGB}
    Alpha {from=MaterialExpressionScalarParameter_3}
   ]
  }
  MaterialExpressionLinearInterpolate_1 {
   class=MaterialExpressionLinearInterpolate
   inputs [
    A {from=MaterialExpressionScalarParameter_0}
    B {from=MaterialExpressionMaterialFunctionCall_1, output=Result}
    Alpha {from=MaterialExpressionScalarParameter_4}
   ]
  }
  MaterialExpressionLinearInterpolate_10 {
   class=MaterialExpressionLinearInterpolate
   inputs [
    A {from=MaterialExpressionVectorParameter_18, output=RGB}
    B {from=MaterialExpressionTextureSampleParameter2D_8, output=RGB}
    Alpha {from=MaterialExpressionScalarParameter_32}
   ]
  }
  MaterialExpressionLinearInterpolate_11 {
   class=MaterialExpressionLinearInterpolate
   inputs [
    A {from=MaterialExpressionScalarParameter_31}
    B {from=MaterialExpressionMaterialFunctionCall_18, output=Result}
    Alpha {from=MaterialExpressionScalarParameter_35}
   ]
  }
  MaterialExpressionLinearInterpolate_12 {
   class=MaterialExpressionLinearInterpolate
   inputs [
    A {from=MaterialExpressionScalarParameter_38}
    B {from=MaterialExpressionMaterialFunctionCall_21, output=Result}
    Alpha {from=MaterialExpressionScalarParameter_39}
   ]
  }
  MaterialExpressionLinearInterpolate_13 {
   class=MaterialExpressionLinearInterpolate
   inputs [
    A {from=MaterialExpressionScalarParameter_43}
    B {from=MaterialExpressionMaterialFunctionCall_24, output=Result}
    Alpha {from=MaterialExpressionScalarParameter_44}
   ]
  }
  MaterialExpressionLinearInterpolate_14 {
   class=MaterialExpressionLinearInterpolate
   inputs [
    A {from=MaterialExpressionScalarParameter_48}
    B {from=MaterialExpressionMaterialFunctionCall_32, output=Result}
    Alpha {from=MaterialExpressionScalarParameter_49}
   ]
  }
  MaterialExpressionLinearInterpolate_2 {
   class=MaterialExpressionLinearInterpolate
   inputs [
    A {from=MaterialExpressionScalarParameter_1}
    B {from=MaterialExpressionMaterialFunctionCall_5, output=Result}
    Alpha {from=MaterialExpressionScalarParameter_5}
   ]
  }
  MaterialExpressionLinearInterpolate_3 {
   class=MaterialExpressionLinearInterpolate
   inputs [
    A {from=MaterialExpressionScalarParameter_2}
    B {from=MaterialExpressionMaterialFunctionCall_3, output=Result}
    Alpha {from=MaterialExpressionScalarParameter_6}
   ]
  }
  MaterialExpressionLinearInterpolate_4 {
   class=MaterialExpressionLinearInterpolate
   inputs [
    A {from=MaterialExpressionVectorParameter_2, output=RGB}
    B {from=MaterialExpressionTextureSampleParameter2D_4, output=RGB}
    Alpha {from=MaterialExpressionScalarParameter_7}
   ]
  }
  MaterialExpressionLinearInterpolate_5 {
   class=MaterialExpressionLinearInterpolate
   inputs [
    A {from=MaterialExpressionVectorParameter_1, output=RGB}
    B {from=MaterialExpressionTextureSampleParameter2D_5, output=RGB}
    Alpha {from=MaterialExpressionScalarParameter_8}
   ]
  }
  MaterialExpressionLinearInterpolate_6 {
   class=MaterialExpressionLinearInterpolate
   inputs [
    A {from=MaterialExpressionReroute_1}
    B {from=MaterialExpressionMaterialFunctionCall_50, output=Result}
    Alpha {from=MaterialExpressionScalarParameter_9}
   ]
  }
  MaterialExpressionLinearInterpolate_7 {
   class=MaterialExpressionLinearInterpolate
   inputs [
    A {from=MaterialExpressionScalarParameter_10}
    B {from=MaterialExpressionMaterialFunctionCall_6, output=Result}
    Alpha {from=MaterialExpressionScalarParameter_11}
   ]
  }
  MaterialExpressionLinearInterpolate_8 {
   class=MaterialExpressionLinearInterpolate
   inputs [
    A {from=MaterialExpressionVectorParameter_0, output=RGB}
    B {from=MaterialExpressionMaterialFunctionCall_13, output=Result}
    Alpha {from=MaterialExpressionScalarParameter_12}
   ]
  }
  MaterialExpressionLinearInterpolate_9 {
   class=MaterialExpressionLinearInterpolate
   inputs [
    A {from=MaterialExpressionScalarParameter_13}
    B {from=MaterialExpressionMaterialFunctionCall_9, output=Result}
    Alpha {from=MaterialExpressionScalarParameter_14}
   ]
  }
  MaterialExpressionMaterialFunctionCall_1 {
   class=MaterialExpressionMaterialFunctionCall
   inputs [
    Value {from=MaterialExpressionTextureSampleParameter2D_1, output=RGBA}
    Channel {from=MaterialExpressionVectorParameter_5, output=RGBA}
   ]
   properties {materialFunction=/InterchangeAssets/Functions/ChannelSelector}
  }
  MaterialExpressionMaterialFunctionCall_10 {
   class=MaterialExpressionMaterialFunctionCall
   inputs [sRGB {from=MaterialExpressionVertexColor_0, output=B}]
   properties {materialFunction=/Engine/Functions/Engine_MaterialFunctions01/ImageAdjustment/sRGBToLinear}
  }
  MaterialExpressionMaterialFunctionCall_11 {
   class=MaterialExpressionMaterialFunctionCall
   inputs [sRGB {from=MaterialExpressionVertexColor_0, output=G}]
   properties {materialFunction=&sRGBToLinear}
  }
  MaterialExpressionMaterialFunctionCall_12 {
   class=MaterialExpressionMaterialFunctionCall
   inputs [sRGB {from=MaterialExpressionVertexColor_0, output=R}]
   properties {materialFunction=&sRGBToLinear}
  }
  MaterialExpressionMaterialFunctionCall_13 {
   class=MaterialExpressionMaterialFunctionCall
   inputs [
    X {from=MaterialExpressionMaterialFunctionCall_12, output=Linear}
    Y {from=MaterialExpressionMaterialFunctionCall_11, output=Linear}
    Z {from=MaterialExpressionMaterialFunctionCall_10, output=Linear}
   ]
   properties {materialFunction=&MakeFloat3}
  }
  MaterialExpressionMaterialFunctionCall_15 {
   class=MaterialExpressionMaterialFunctionCall
   inputs [
    diffuseColor {from=MaterialExpressionLinearInterpolate_0}
    emissiveColor {from=MaterialExpressionLinearInterpolate_5}
    specularColor {from=MaterialExpressionLinearInterpolate_10}
    metallic {from=MaterialExpressionLinearInterpolate_1}
    roughness {from=MaterialExpressionLinearInterpolate_2}
    clearcoat {from=MaterialExpressionLinearInterpolate_11}
    clearcoatRoughness {from=MaterialExpressionLinearInterpolate_12}
    opacity {from=MaterialExpressionLinearInterpolate_13}
    opacityMode {from=MaterialExpressionScalarParameter_47}
    opacityThreshold {from=MaterialExpressionLinearInterpolate_3}
    ior {from=MaterialExpressionReroute_0}
    normal {from=MaterialExpressionLinearInterpolate_6}
    displacement {from=MaterialExpressionReroute_5}
    useSpecularWorkflow {from=MaterialExpressionScalarParameter_42}
    occlusion {from=MaterialExpressionReroute_2}
   ]
   properties {materialFunction=/Engine/Functions/Substrate/MF_Substrate_UsdPreviewSurface}
  }
  MaterialExpressionMaterialFunctionCall_18 {
   class=MaterialExpressionMaterialFunctionCall
   inputs [
    Value {from=MaterialExpressionTextureSampleParameter2D_9, output=RGBA}
    Channel {from=MaterialExpressionVectorParameter_16, output=RGBA}
   ]
   properties {materialFunction=&ChannelSelector}
  }
  MaterialExpressionMaterialFunctionCall_21 {
   class=MaterialExpressionMaterialFunctionCall
   inputs [
    Value {from=MaterialExpressionTextureSampleParameter2D_10, output=RGBA}
    Channel {from=MaterialExpressionVectorParameter_20, output=RGBA}
   ]
   properties {materialFunction=&ChannelSelector}
  }
  MaterialExpressionMaterialFunctionCall_24 {
   class=MaterialExpressionMaterialFunctionCall
   inputs [
    Value {from=MaterialExpressionTextureSampleParameter2D_11, output=RGBA}
    Channel {from=MaterialExpressionVectorParameter_22, output=RGBA}
   ]
   properties {materialFunction=&ChannelSelector}
  }
  MaterialExpressionMaterialFunctionCall_3 {
   class=MaterialExpressionMaterialFunctionCall
   inputs [
    Value {from=MaterialExpressionTextureSampleParameter2D_3, output=RGBA}
    Channel {from=MaterialExpressionVectorParameter_6, output=RGBA}
   ]
   properties {materialFunction=&ChannelSelector}
  }
  MaterialExpressionMaterialFunctionCall_32 {
   class=MaterialExpressionMaterialFunctionCall
   inputs [
    Value {from=MaterialExpressionTextureSampleParameter2D_12, output=RGBA}
    Channel {from=MaterialExpressionVectorParameter_24, output=RGBA}
   ]
   properties {materialFunction=&ChannelSelector}
  }
  MaterialExpressionMaterialFunctionCall_33 {
   class=MaterialExpressionMaterialFunctionCall
   inputs [
    Scale {from=MaterialExpressionConvert_2}
    Rotation {from=MaterialExpressionScalarParameter_23}
    TextureCoordinate {from=MaterialExpressionReroute_25, output=UV}
    Translation {from=MaterialExpressionConvert_2}
   ]
   properties {materialFunction=/InterchangeAssets/Functions/UsdTransform2d}
  }
  MaterialExpressionMaterialFunctionCall_35 {
   class=MaterialExpressionMaterialFunctionCall
   inputs [
    Scale {from=MaterialExpressionConvert_3}
    Rotation {from=MaterialExpressionScalarParameter_24}
    TextureCoordinate {from=MaterialExpressionReroute_1, output=UV}
    Translation {from=MaterialExpressionConvert_3}
   ]
   properties {materialFunction=&UsdTransform2d}
  }
  MaterialExpressionMaterialFunctionCall_36 {
   class=MaterialExpressionMaterialFunctionCall
   inputs [
    Scale {from=MaterialExpressionConvert_4}
    Rotation {from=MaterialExpressionScalarParameter_25}
    TextureCoordinate {from=MaterialExpressionReroute_1, output=UV}
    Translation {from=MaterialExpressionConvert_4}
   ]
   properties {materialFunction=&UsdTransform2d}
  }
  MaterialExpressionMaterialFunctionCall_38 {
   class=MaterialExpressionMaterialFunctionCall
   inputs [
    Scale {from=MaterialExpressionConvert_1}
    Rotation {from=MaterialExpressionScalarParameter_26}
    TextureCoordinate {from=MaterialExpressionReroute_21, output=UV}
    Translation {from=MaterialExpressionConvert_1}
   ]
   properties {materialFunction=&UsdTransform2d}
  }
  MaterialExpressionMaterialFunctionCall_39 {
   class=MaterialExpressionMaterialFunctionCall
   inputs [
    Scale {from=MaterialExpressionConvert_5}
    Rotation {from=MaterialExpressionScalarParameter_27}
    TextureCoordinate {from=MaterialExpressionReroute_1, output=UV}
    Translation {from=MaterialExpressionConvert_5}
   ]
   properties {materialFunction=&UsdTransform2d}
  }
  MaterialExpressionMaterialFunctionCall_41 {
   class=MaterialExpressionMaterialFunctionCall
   inputs [
    Scale {from=MaterialExpressionConvert_6}
    Rotation {from=MaterialExpressionScalarParameter_28}
    TextureCoordinate {from=MaterialExpressionReroute_27, output=UV}
    Translation {from=MaterialExpressionConvert_6}
   ]
   properties {materialFunction=&UsdTransform2d}
  }
  MaterialExpressionMaterialFunctionCall_42 {
   class=MaterialExpressionMaterialFunctionCall
   inputs [
    Scale {from=MaterialExpressionConvert_7}
    Rotation {from=MaterialExpressionScalarParameter_29}
    TextureCoordinate {from=MaterialExpressionReroute_29, output=UV}
    Translation {from=MaterialExpressionConvert_7}
   ]
   properties {materialFunction=&UsdTransform2d}
  }
  MaterialExpressionMaterialFunctionCall_43 {
   class=MaterialExpressionMaterialFunctionCall
   inputs [
    Scale {from=MaterialExpressionConvert_8}
    Rotation {from=MaterialExpressionScalarParameter_30}
    TextureCoordinate {from=MaterialExpressionReroute_1, output=UV}
    Translation {from=MaterialExpressionConvert_8}
   ]
   properties {materialFunction=&UsdTransform2d}
  }
  MaterialExpressionMaterialFunctionCall_45 {
   class=MaterialExpressionMaterialFunctionCall
   inputs [
    Scale {from=MaterialExpressionConvert_0}
    Rotation {from=MaterialExpressionScalarParameter_34}
    TextureCoordinate {from=MaterialExpressionReroute_17, output=UV}
    Translation {from=MaterialExpressionConvert_0}
   ]
   properties {materialFunction=&UsdTransform2d}
  }
  MaterialExpressionMaterialFunctionCall_46 {
   class=MaterialExpressionMaterialFunctionCall
   inputs [
    Scale {from=MaterialExpressionConvert_9}
    Rotation {from=MaterialExpressionScalarParameter_37}
    TextureCoordinate {from=MaterialExpressionReroute_3, output=UV}
    Translation {from=MaterialExpressionConvert_9}
   ]
   properties {materialFunction=&UsdTransform2d}
  }
  MaterialExpressionMaterialFunctionCall_47 {
   class=MaterialExpressionMaterialFunctionCall
   inputs [
    Scale {from=MaterialExpressionConvert_10}
    Rotation {from=MaterialExpressionScalarParameter_41}
    TextureCoordinate {from=MaterialExpressionReroute_4, output=UV}
    Translation {from=MaterialExpressionConvert_10}
   ]
   properties {materialFunction=&UsdTransform2d}
  }
  MaterialExpressionMaterialFunctionCall_48 {
   class=MaterialExpressionMaterialFunctionCall
   inputs [
    Scale {from=MaterialExpressionConvert_11}
    Rotation {from=MaterialExpressionScalarParameter_46}
    TextureCoordinate {from=MaterialExpressionReroute_25, output=UV}
    Translation {from=MaterialExpressionConvert_11}
   ]
   properties {materialFunction=&UsdTransform2d}
  }
  MaterialExpressionMaterialFunctionCall_49 {
   class=MaterialExpressionMaterialFunctionCall
   inputs [
    Scale {from=MaterialExpressionConvert_12}
    Rotation {from=MaterialExpressionScalarParameter_51}
    TextureCoordinate {from=MaterialExpressionReroute_27, output=UV}
    Translation {from=MaterialExpressionConvert_12}
   ]
   properties {materialFunction=&UsdTransform2d}
  }
  MaterialExpressionMaterialFunctionCall_5 {
   class=MaterialExpressionMaterialFunctionCall
   inputs [
    Value {from=MaterialExpressionTextureSampleParameter2D_2, output=RGBA}
    Channel {from=MaterialExpressionVectorParameter_7, output=RGBA}
   ]
   properties {materialFunction=&ChannelSelector}
  }
  MaterialExpressionMaterialFunctionCall_50 {
   class=MaterialExpressionMaterialFunctionCall
   inputs [In {from=MaterialExpressionReroute_1}]
   properties {materialFunction=/InterchangeAssets/Functions/WorldSpaceNormalsToTangentSpace}
  }
  MaterialExpressionMaterialFunctionCall_6 {
   class=MaterialExpressionMaterialFunctionCall
   inputs [
    Value {from=MaterialExpressionTextureSampleParameter2D_6, output=RGBA}
    Channel {from=MaterialExpressionVectorParameter_3, output=RGBA}
   ]
   properties {materialFunction=&ChannelSelector}
  }
  MaterialExpressionMaterialFunctionCall_9 {
   class=MaterialExpressionMaterialFunctionCall
   inputs [
    Value {from=MaterialExpressionTextureSampleParameter2D_7, output=RGBA}
    Channel {from=MaterialExpressionVectorParameter_4, output=RGBA}
   ]
   properties {materialFunction=&ChannelSelector}
  }
  MaterialExpressionPinBase_0 {
   class=MaterialExpressionPinBase
   inputs []
   properties {reroutePins [UVIndex {expression=&self:MaterialExpressionReroute_16}]}
  }
  MaterialExpressionPinBase_1 {
   class=MaterialExpressionPinBase
   inputs [UV {from=MaterialExpressionStaticSwitchParameter_0, output=""}]
   properties {reroutePins [UV {expression=&self:MaterialExpressionReroute_17}]}
  }
  MaterialExpressionPinBase_14 {
   class=MaterialExpressionPinBase
   inputs []
   properties {reroutePins [UVIndex {expression=&self:MaterialGraph_0.MaterialGraphNode_Composite_12.UVSwitch.MaterialExpressionReroute_24}]}
  }
  MaterialExpressionPinBase_15 {
   class=MaterialExpressionPinBase
   inputs [UV {from=MaterialExpressionStaticSwitchParameter_1, output=""}]
   properties {reroutePins [UV {expression=&self:MaterialGraph_0.MaterialGraphNode_Composite_12.UVSwitch.MaterialExpressionReroute_25}]}
  }
  MaterialExpressionPinBase_2 {
   class=MaterialExpressionPinBase
   inputs []
   properties {reroutePins [UVIndex {expression=&self:MaterialGraph_2.MaterialGraphNode_Composite_10.UVSwitch.MaterialExpressionReroute_2}]}
  }
  MaterialExpressionPinBase_24 {
   class=MaterialExpressionPinBase
   inputs []
   properties {reroutePins [UVIndex {expression=&self:MaterialGraph_1.MaterialGraphNode_Composite_8.UVSwitch.MaterialExpressionReroute_0}]}
  }
  MaterialExpressionPinBase_25 {
   class=MaterialExpressionPinBase
   inputs [UV {from=MaterialExpressionStaticSwitchParameter_17, output=""}]
   properties {reroutePins [UV {expression=&self:MaterialGraph_1.MaterialGraphNode_Composite_8.UVSwitch.MaterialExpressionReroute_1}]}
  }
  MaterialExpressionPinBase_26 {
   class=MaterialExpressionPinBase
   inputs []
   properties {reroutePins [UVIndex {expression=&self:MaterialGraph_2.MaterialGraphNode_Composite_9.UVSwitch.MaterialExpressionReroute_0}]}
  }
  MaterialExpressionPinBase_27 {
   class=MaterialExpressionPinBase
   inputs [UV {from=MaterialExpressionStaticSwitchParameter_20, output=""}]
   properties {reroutePins [UV {expression=&self:MaterialGraph_2.MaterialGraphNode_Composite_9.UVSwitch.MaterialExpressionReroute_1}]}
  }
  MaterialExpressionPinBase_28 {
   class=MaterialExpressionPinBase
   inputs []
   properties {reroutePins [UVIndex {expression=&self:MaterialGraph_0.MaterialGraphNode_Composite_15.UVSwitch.MaterialExpressionReroute_20}]}
  }
  MaterialExpressionPinBase_29 {
   class=MaterialExpressionPinBase
   inputs [UV {from=MaterialExpressionStaticSwitchParameter_23, output=""}]
   properties {reroutePins [UV {expression=&self:MaterialGraph_0.MaterialGraphNode_Composite_15.UVSwitch.MaterialExpressionReroute_21}]}
  }
  MaterialExpressionPinBase_3 {
   class=MaterialExpressionPinBase
   inputs [UV {from=MaterialExpressionStaticSwitchParameter_2, output=""}]
   properties {reroutePins [UV {expression=&self:MaterialGraph_2.MaterialGraphNode_Composite_10.UVSwitch.MaterialExpressionReroute_3}]}
  }
  MaterialExpressionPinBase_30 {
   class=MaterialExpressionPinBase
   inputs []
   properties {reroutePins [UVIndex {expression=&self:MaterialGraph_2.MaterialGraphNode_Composite_11.UVSwitch.MaterialExpressionReroute_0}]}
  }
  MaterialExpressionPinBase_31 {
   class=MaterialExpressionPinBase
   inputs [UV {from=MaterialExpressionStaticSwitchParameter_26, output=""}]
   properties {reroutePins [UV {expression=&self:MaterialGraph_2.MaterialGraphNode_Composite_11.UVSwitch.MaterialExpressionReroute_1}]}
  }
  MaterialExpressionPinBase_32 {
   class=MaterialExpressionPinBase
   inputs []
   properties {reroutePins [UVIndex {expression=&self:MaterialGraph_0.MaterialGraphNode_Composite_17.UVSwitch.MaterialExpressionReroute_26}]}
  }
  MaterialExpressionPinBase_33 {
   class=MaterialExpressionPinBase
   inputs [UV {from=MaterialExpressionStaticSwitchParameter_29, output=""}]
   properties {reroutePins [UV {expression=&self:MaterialGraph_0.MaterialGraphNode_Composite_17.UVSwitch.MaterialExpressionReroute_27}]}
  }
  MaterialExpressionPinBase_34 {
   class=MaterialExpressionPinBase
   inputs []
   properties {reroutePins [UVIndex {expression=&self:MaterialGraph_0.MaterialGraphNode_Composite_18.UVSwitch.MaterialExpressionReroute_28}]}
  }
  MaterialExpressionPinBase_35 {
   class=MaterialExpressionPinBase
   inputs [UV {from=MaterialExpressionStaticSwitchParameter_32, output=""}]
   properties {reroutePins [UV {expression=&self:MaterialGraph_0.MaterialGraphNode_Composite_18.UVSwitch.MaterialExpressionReroute_29}]}
  }
  MaterialExpressionPinBase_36 {
   class=MaterialExpressionPinBase
   inputs []
   properties {reroutePins [UVIndex {expression=&self:MaterialGraph_2.MaterialGraphNode_Composite_12.UVSwitch.MaterialExpressionReroute_0}]}
  }
  MaterialExpressionPinBase_37 {
   class=MaterialExpressionPinBase
   inputs [UV {from=MaterialExpressionStaticSwitchParameter_35, output=""}]
   properties {reroutePins [UV {expression=&self:MaterialGraph_2.MaterialGraphNode_Composite_12.UVSwitch.MaterialExpressionReroute_1}]}
  }
  MaterialExpressionPinBase_4 {
   class=MaterialExpressionPinBase
   inputs []
   properties {reroutePins [UVIndex {expression=&self:MaterialExpressionReroute_3}]}
  }
  MaterialExpressionPinBase_5 {
   class=MaterialExpressionPinBase
   inputs [UV {from=MaterialExpressionStaticSwitchParameter_3, output=""}]
   properties {reroutePins [UV {expression=&self:MaterialExpressionReroute_4}]}
  }
  MaterialExpressionPinBase_6 {
   class=MaterialExpressionPinBase
   inputs []
   properties {reroutePins [UVIndex {expression=&self:MaterialExpressionReroute_24}]}
  }
  MaterialExpressionPinBase_7 {
   class=MaterialExpressionPinBase
   inputs [UV {from=MaterialExpressionStaticSwitchParameter_4, output=""}]
   properties {reroutePins [UV {expression=&self:MaterialExpressionReroute_25}]}
  }
  MaterialExpressionPinBase_8 {
   class=MaterialExpressionPinBase
   inputs []
   properties {reroutePins [UVIndex {expression=&self:MaterialExpressionReroute_26}]}
  }
  MaterialExpressionPinBase_9 {
   class=MaterialExpressionPinBase
   inputs [UV {from=MaterialExpressionStaticSwitchParameter_5, output=""}]
   properties {reroutePins [UV {expression=&self:MaterialExpressionReroute_27}]}
  }
  MaterialExpressionReroute_0 {class=MaterialExpressionReroute, inputs [None {from=MaterialExpressionLinearInterpolate_7}]}
  MaterialExpressionReroute_1 {class=MaterialExpressionReroute, inputs [None {from=MaterialExpressionLinearInterpolate_4}]}
  MaterialExpressionReroute_16 {class=MaterialExpressionReroute, inputs [None {from=MaterialExpressionScalarParameter_33}]}
  MaterialExpressionReroute_17 {class=MaterialExpressionReroute, inputs [None {from=MaterialExpressionStaticSwitchParameter_0}]}
  MaterialExpressionReroute_2 {class=MaterialExpressionReroute, inputs [None {from=MaterialExpressionLinearInterpolate_9}]}
  MaterialExpressionReroute_24 {class=MaterialExpressionReroute, inputs [None {from=MaterialExpressionScalarParameter_45}]}
  MaterialExpressionReroute_25 {class=MaterialExpressionReroute, inputs [None {from=MaterialExpressionStaticSwitchParameter_4}]}
  MaterialExpressionReroute_26 {class=MaterialExpressionReroute, inputs [None {from=MaterialExpressionScalarParameter_50}]}
  MaterialExpressionReroute_27 {class=MaterialExpressionReroute, inputs [None {from=MaterialExpressionStaticSwitchParameter_5}]}
  MaterialExpressionReroute_3 {class=MaterialExpressionReroute, inputs [None {from=MaterialExpressionScalarParameter_40}]}
  MaterialExpressionReroute_4 {class=MaterialExpressionReroute, inputs [None {from=MaterialExpressionStaticSwitchParameter_3}]}
  MaterialExpressionReroute_5 {class=MaterialExpressionReroute, inputs [None {from=MaterialExpressionLinearInterpolate_14}]}
  MaterialExpressionScalarParameter_0 {class=MaterialExpressionScalarParameter, properties {parameterName=Metallic, sliderMax=1}}
  MaterialExpressionScalarParameter_1 {class=MaterialExpressionScalarParameter, properties {defaultValue=0.5, parameterName=Roughness, sliderMax=1}}
  MaterialExpressionScalarParameter_10 {class=MaterialExpressionScalarParameter, properties {defaultValue=1.5, parameterName=Refraction}}
  MaterialExpressionScalarParameter_11 {class=MaterialExpressionScalarParameter, properties {group=UseTexture, parameterName=UseRefractionTexture}}
  MaterialExpressionScalarParameter_12 {class=MaterialExpressionScalarParameter, properties {group=Global, parameterName=UseVertexColorForBaseColor}}
  MaterialExpressionScalarParameter_13 {class=MaterialExpressionScalarParameter, properties {defaultValue=1, parameterName=AmbientOcclusion, sliderMax=1}}
  MaterialExpressionScalarParameter_14 {class=MaterialExpressionScalarParameter, properties {group=UseTexture, parameterName=UseAmbientOcclusionTexture}}
  MaterialExpressionScalarParameter_15 {class=MaterialExpressionScalarParameter, properties {group=UVIndex, parameterName=BaseColorUVIndex}}
  MaterialExpressionScalarParameter_16 {class=MaterialExpressionScalarParameter, properties {group=UVIndex, parameterName=MetallicUVIndex}}
  MaterialExpressionScalarParameter_17 {class=MaterialExpressionScalarParameter, properties {group=UVIndex, parameterName=RoughnessUVIndex}}
  MaterialExpressionScalarParameter_18 {class=MaterialExpressionScalarParameter, properties {group=UVIndex, parameterName=EmissiveUVIndex}}
  MaterialExpressionScalarParameter_19 {class=MaterialExpressionScalarParameter, properties {group=UVIndex, parameterName=OpacityThresholdUVIndex}}
  MaterialExpressionScalarParameter_2 {class=MaterialExpressionScalarParameter, properties {parameterName=OpacityThreshold, sliderMax=1}}
  MaterialExpressionScalarParameter_20 {class=MaterialExpressionScalarParameter, properties {group=UVIndex, parameterName=NormalUVIndex}}
  MaterialExpressionScalarParameter_21 {class=MaterialExpressionScalarParameter, properties {group=UVIndex, parameterName=RefractionUVIndex}}
  MaterialExpressionScalarParameter_22 {class=MaterialExpressionScalarParameter, properties {group=UVIndex, parameterName=AmbientOcclusionUVIndex}}
  MaterialExpressionScalarParameter_23 {class=MaterialExpressionScalarParameter, properties {group=Rotation, parameterName=BaseColorRotation}}
  MaterialExpressionScalarParameter_24 {class=MaterialExpressionScalarParameter, properties {group=Rotation, parameterName=MetallicRotation}}
  MaterialExpressionScalarParameter_25 {class=MaterialExpressionScalarParameter, properties {group=Rotation, parameterName=RoughnessRotation}}
  MaterialExpressionScalarParameter_26 {class=MaterialExpressionScalarParameter, properties {group=Rotation, parameterName=EmissiveRotation}}
  MaterialExpressionScalarParameter_27 {class=MaterialExpressionScalarParameter, properties {group=Rotation, parameterName=OpacityThresholdRotation}}
  MaterialExpressionScalarParameter_28 {class=MaterialExpressionScalarParameter, properties {group=Rotation, parameterName=NormalRotation}}
  MaterialExpressionScalarParameter_29 {class=MaterialExpressionScalarParameter, properties {group=Rotation, parameterName=RefractionRotation}}
  MaterialExpressionScalarParameter_3 {class=MaterialExpressionScalarParameter, properties {group=UseTexture, parameterName=UseBaseColorTexture}}
  MaterialExpressionScalarParameter_30 {class=MaterialExpressionScalarParameter, properties {group=Rotation, parameterName=AmbientOcclusionRotation}}
  MaterialExpressionScalarParameter_31 {class=MaterialExpressionScalarParameter, properties {parameterName=ClearCoat, sliderMax=1}}
  MaterialExpressionScalarParameter_32 {class=MaterialExpressionScalarParameter, properties {group=UseTexture, parameterName=UseSpecularColorTexture}}
  MaterialExpressionScalarParameter_33 {class=MaterialExpressionScalarParameter, properties {group=UVIndex, parameterName=SpecularUVIndex}}
  MaterialExpressionScalarParameter_34 {class=MaterialExpressionScalarParameter, properties {group=Rotation, parameterName=SpecularRotation}}
  MaterialExpressionScalarParameter_35 {class=MaterialExpressionScalarParameter, properties {group=UseTexture, parameterName=UseClearCoatTexture}}
  MaterialExpressionScalarParameter_36 {class=MaterialExpressionScalarParameter, properties {group=UVIndex, parameterName=ClearCoatUVIndex}}
  MaterialExpressionScalarParameter_37 {class=MaterialExpressionScalarParameter, properties {group=Rotation, parameterName=ClearCoatRotation}}
  MaterialExpressionScalarParameter_38 {class=MaterialExpressionScalarParameter, properties {defaultValue=0.01, parameterName=ClearCoatRoughness, sliderMax=1}}
  MaterialExpressionScalarParameter_39 {class=MaterialExpressionScalarParameter, properties {group=UseTexture, parameterName=UseClearCoatRoughnessTexture}}
  MaterialExpressionScalarParameter_4 {class=MaterialExpressionScalarParameter, properties {group=UseTexture, parameterName=UseMetallicTexture}}
  MaterialExpressionScalarParameter_40 {class=MaterialExpressionScalarParameter, properties {group=UVIndex, parameterName=ClearCoatRoughnessUVIndex}}
  MaterialExpressionScalarParameter_41 {class=MaterialExpressionScalarParameter, properties {group=Rotation, parameterName=ClearCoatRoughnessRotation}}
  MaterialExpressionScalarParameter_42 {
   class=MaterialExpressionScalarParameter
   properties {
    controlType=Enumeration
    enumeration=/InterchangeAssets/Utilities/EUsdPreviewSurfaceWorkflow
    enumerationIndex=1
    group=Global
    parameterName=UseSpecularWorkflow
    sliderMax=1
   }
  }
  MaterialExpressionScalarParameter_43 {class=MaterialExpressionScalarParameter, properties {defaultValue=1, parameterName=Opacity, sliderMax=1}}
  MaterialExpressionScalarParameter_44 {class=MaterialExpressionScalarParameter, properties {group=UseTexture, parameterName=UseOpacityTexture}}
  MaterialExpressionScalarParameter_45 {class=MaterialExpressionScalarParameter, properties {group=UVIndex, parameterName=OpacityUVIndex}}
  MaterialExpressionScalarParameter_46 {class=MaterialExpressionScalarParameter, properties {group=Rotation, parameterName=OpacityRotation}}
  MaterialExpressionScalarParameter_47 {
   class=MaterialExpressionScalarParameter
   properties {controlType=Enumeration, enumeration=/InterchangeAssets/Utilities/EUsdPreviewSurfaceOpacityMode, group=Global, parameterName=OpacityMode}
  }
  MaterialExpressionScalarParameter_48 {class=MaterialExpressionScalarParameter, properties {parameterName=Displacement, sliderMax=1}}
  MaterialExpressionScalarParameter_49 {class=MaterialExpressionScalarParameter, properties {group=UseTexture, parameterName=UseDisplacementTexture}}
  MaterialExpressionScalarParameter_5 {class=MaterialExpressionScalarParameter, properties {group=UseTexture, parameterName=UseRoughnessTexture}}
  MaterialExpressionScalarParameter_50 {class=MaterialExpressionScalarParameter, properties {group=UVIndex, parameterName=DisplacementUVIndex}}
  MaterialExpressionScalarParameter_51 {class=MaterialExpressionScalarParameter, properties {group=Rotation, parameterName=DisplacementRotation}}
  MaterialExpressionScalarParameter_6 {class=MaterialExpressionScalarParameter, properties {group=UseTexture, parameterName=UseOpacityThresholdTexture}}
  MaterialExpressionScalarParameter_7 {class=MaterialExpressionScalarParameter, properties {group=UseTexture, parameterName=UseNormalTexture}}
  MaterialExpressionScalarParameter_8 {class=MaterialExpressionScalarParameter, properties {group=UseTexture, parameterName=UseEmissiveColorTexture}}
  MaterialExpressionScalarParameter_9 {
   class=MaterialExpressionScalarParameter
   properties {desc=Not a static switch so that we can do this at runtime, group=Global, parameterName=UseWorldSpaceNormals}
  }
  MaterialExpressionStaticSwitchParameter_0 {
   class=MaterialExpressionStaticSwitchParameter
   inputs [
    True {from=MaterialExpressionSwitch_1, output=""}
    False {from=MaterialExpressionIf_0, output=""}
   ]
   properties {
    desc=Fallback to use 'if' nodes instead of switch expression. Safer way to have analytical derivatives with nanite
    group=Nanite
    parameterName=bUseSwitch
   }
  }
  MaterialExpressionStaticSwitchParameter_1 {
   class=MaterialExpressionStaticSwitchParameter
   inputs [
    True {from=MaterialExpressionSwitch_0, output=""}
    False {from=MaterialExpressionIf_22, output=""}
   ]
   properties {
    desc=Fallback to use 'if' nodes instead of switch expression. Safer way to have analytical derivatives with nanite
    group=Nanite
    parameterName=bUseSwitch
   }
  }
  MaterialExpressionStaticSwitchParameter_17 {
   class=MaterialExpressionStaticSwitchParameter
   inputs [
    True {from=MaterialExpressionSwitch_5, output=""}
    False {from=MaterialExpressionIf_37, output=""}
   ]
   properties {
    desc=Fallback to use 'if' nodes instead of switch expression. Safer way to have analytical derivatives with nanite
    group=Nanite
    parameterName=bUseSwitch
   }
  }
  MaterialExpressionStaticSwitchParameter_2 {
   class=MaterialExpressionStaticSwitchParameter
   inputs [
    True {from=MaterialExpressionSwitch_2, output=""}
    False {from=MaterialExpressionIf_3, output=""}
   ]
   properties {
    desc=Fallback to use 'if' nodes instead of switch expression. Safer way to have analytical derivatives with nanite
    group=Nanite
    parameterName=bUseSwitch
   }
  }
  MaterialExpressionStaticSwitchParameter_20 {
   class=MaterialExpressionStaticSwitchParameter
   inputs [
    True {from=MaterialExpressionSwitch_6, output=""}
    False {from=MaterialExpressionIf_40, output=""}
   ]
   properties {
    desc=Fallback to use 'if' nodes instead of switch expression. Safer way to have analytical derivatives with nanite
    group=Nanite
    parameterName=bUseSwitch
   }
  }
  MaterialExpressionStaticSwitchParameter_23 {
   class=MaterialExpressionStaticSwitchParameter
   inputs [
    True {from=MaterialExpressionSwitch_7, output=""}
    False {from=MaterialExpressionIf_43, output=""}
   ]
   properties {
    desc=Fallback to use 'if' nodes instead of switch expression. Safer way to have analytical derivatives with nanite
    group=Nanite
    parameterName=bUseSwitch
   }
  }
  MaterialExpressionStaticSwitchParameter_26 {
   class=MaterialExpressionStaticSwitchParameter
   inputs [
    True {from=MaterialExpressionSwitch_8, output=""}
    False {from=MaterialExpressionIf_46, output=""}
   ]
   properties {
    desc=Fallback to use 'if' nodes instead of switch expression. Safer way to have analytical derivatives with nanite
    group=Nanite
    parameterName=bUseSwitch
   }
  }
  MaterialExpressionStaticSwitchParameter_29 {
   class=MaterialExpressionStaticSwitchParameter
   inputs [
    True {from=MaterialExpressionSwitch_9, output=""}
    False {from=MaterialExpressionIf_49, output=""}
   ]
   properties {
    desc=Fallback to use 'if' nodes instead of switch expression. Safer way to have analytical derivatives with nanite
    group=Nanite
    parameterName=bUseSwitch
   }
  }
  MaterialExpressionStaticSwitchParameter_3 {
   class=MaterialExpressionStaticSwitchParameter
   inputs [
    True {from=MaterialExpressionSwitch_3, output=""}
    False {from=MaterialExpressionIf_6, output=""}
   ]
   properties {
    desc=Fallback to use 'if' nodes instead of switch expression. Safer way to have analytical derivatives with nanite
    group=Nanite
    parameterName=bUseSwitch
   }
  }
  MaterialExpressionStaticSwitchParameter_32 {
   class=MaterialExpressionStaticSwitchParameter
   inputs [
    True {from=MaterialExpressionSwitch_10, output=""}
    False {from=MaterialExpressionIf_52, output=""}
   ]
   properties {
    desc=Fallback to use 'if' nodes instead of switch expression. Safer way to have analytical derivatives with nanite
    group=Nanite
    parameterName=bUseSwitch
   }
  }
  MaterialExpressionStaticSwitchParameter_35 {
   class=MaterialExpressionStaticSwitchParameter
   inputs [
    True {from=MaterialExpressionSwitch_11, output=""}
    False {from=MaterialExpressionIf_55, output=""}
   ]
   properties {
    desc=Fallback to use 'if' nodes instead of switch expression. Safer way to have analytical derivatives with nanite
    group=Nanite
    parameterName=bUseSwitch
   }
  }
  MaterialExpressionStaticSwitchParameter_4 {
   class=MaterialExpressionStaticSwitchParameter
   inputs [
    True {from=MaterialExpressionSwitch_4, output=""}
    False {from=MaterialExpressionIf_9, output=""}
   ]
   properties {
    desc=Fallback to use 'if' nodes instead of switch expression. Safer way to have analytical derivatives with nanite
    group=Nanite
    parameterName=bUseSwitch
   }
  }
  MaterialExpressionStaticSwitchParameter_5 {
   class=MaterialExpressionStaticSwitchParameter
   inputs [
    True {from=MaterialExpressionSwitch_12, output=""}
    False {from=MaterialExpressionIf_12, output=""}
   ]
   properties {
    desc=Fallback to use 'if' nodes instead of switch expression. Safer way to have analytical derivatives with nanite
    group=Nanite
    parameterName=bUseSwitch
   }
  }
  MaterialExpressionSwitch_0 {
   class=MaterialExpressionSwitch
   inputs [
    SwitchValue {from=MaterialExpressionReroute_24, output=UVIndex}
    Default {from=MaterialExpressionTextureCoordinate_32, output=""}
    UV0 {from=MaterialExpressionTextureCoordinate_32, output=""}
    UV1 {from=MaterialExpressionTextureCoordinate_33, output=""}
    UV2 {from=MaterialExpressionTextureCoordinate_34, output=""}
    UV3 {from=MaterialExpressionTextureCoordinate_35, output=""}
   ]
   properties {inputs [
     {
      inputName=UV0
      input {
       expression=&self:MaterialExpressionTextureCoordinate_32
       inputName=None
       mask=0
       maskA=0
       maskB=0
       maskG=0
       maskR=0
       outputIndex=0
      }
     }
     {
      inputName=UV1
      input {
       expression=&self:MaterialExpressionTextureCoordinate_33
       inputName=None
       mask=0
       maskA=0
       maskB=0
       maskG=0
       maskR=0
       outputIndex=0
      }
     }
     {
      inputName=UV2
      input {
       expression=&self:MaterialExpressionTextureCoordinate_34
       inputName=None
       mask=0
       maskA=0
       maskB=0
       maskG=0
       maskR=0
       outputIndex=0
      }
     }
     {
      inputName=UV3
      input {
       expression=&self:MaterialExpressionTextureCoordinate_35
       inputName=None
       mask=0
       maskA=0
       maskB=0
       maskG=0
       maskR=0
       outputIndex=0
      }
     }
    ]}
  }
  MaterialExpressionSwitch_1 {
   class=MaterialExpressionSwitch
   inputs [
    SwitchValue {from=MaterialExpressionReroute_16, output=UVIndex}
    Default {from=MaterialExpressionTextureCoordinate_0, output=""}
    UV0 {from=MaterialExpressionTextureCoordinate_0, output=""}
    UV1 {from=MaterialExpressionTextureCoordinate_1, output=""}
    UV2 {from=MaterialExpressionTextureCoordinate_2, output=""}
    UV3 {from=MaterialExpressionTextureCoordinate_3, output=""}
   ]
   properties {inputs [
     {
      inputName=UV0
      input {
       expression=&self:MaterialExpressionTextureCoordinate_0
       inputName=None
       mask=0
       maskA=0
       maskB=0
       maskG=0
       maskR=0
       outputIndex=0
      }
     }
     {
      inputName=UV1
      input {
       expression=&self:MaterialExpressionTextureCoordinate_1
       inputName=None
       mask=0
       maskA=0
       maskB=0
       maskG=0
       maskR=0
       outputIndex=0
      }
     }
     {
      inputName=UV2
      input {
       expression=&self:MaterialExpressionTextureCoordinate_2
       inputName=None
       mask=0
       maskA=0
       maskB=0
       maskG=0
       maskR=0
       outputIndex=0
      }
     }
     {
      inputName=UV3
      input {
       expression=&self:MaterialExpressionTextureCoordinate_3
       inputName=None
       mask=0
       maskA=0
       maskB=0
       maskG=0
       maskR=0
       outputIndex=0
      }
     }
    ]}
  }
  MaterialExpressionSwitch_10 {
   class=MaterialExpressionSwitch
   inputs [
    SwitchValue {from=MaterialExpressionReroute_28, output=UVIndex}
    Default {from=MaterialExpressionTextureCoordinate_72, output=""}
    UV0 {from=MaterialExpressionTextureCoordinate_72, output=""}
    UV1 {from=MaterialExpressionTextureCoordinate_73, output=""}
    UV2 {from=MaterialExpressionTextureCoordinate_74, output=""}
    UV3 {from=MaterialExpressionTextureCoordinate_75, output=""}
   ]
   properties {inputs [
     {
      inputName=UV0
      input {
       expression=&self:MaterialExpressionTextureCoordinate_72
       inputName=None
       mask=0
       maskA=0
       maskB=0
       maskG=0
       maskR=0
       outputIndex=0
      }
     }
     {
      inputName=UV1
      input {
       expression=&self:MaterialExpressionTextureCoordinate_73
       inputName=None
       mask=0
       maskA=0
       maskB=0
       maskG=0
       maskR=0
       outputIndex=0
      }
     }
     {
      inputName=UV2
      input {
       expression=&self:MaterialExpressionTextureCoordinate_74
       inputName=None
       mask=0
       maskA=0
       maskB=0
       maskG=0
       maskR=0
       outputIndex=0
      }
     }
     {
      inputName=UV3
      input {
       expression=&self:MaterialExpressionTextureCoordinate_75
       inputName=None
       mask=0
       maskA=0
       maskB=0
       maskG=0
       maskR=0
       outputIndex=0
      }
     }
    ]}
  }
  MaterialExpressionSwitch_11 {
   class=MaterialExpressionSwitch
   inputs [
    SwitchValue {from=MaterialExpressionReroute_0, output=UVIndex}
    Default {from=MaterialExpressionTextureCoordinate_76, output=""}
    UV0 {from=MaterialExpressionTextureCoordinate_76, output=""}
    UV1 {from=MaterialExpressionTextureCoordinate_77, output=""}
    UV2 {from=MaterialExpressionTextureCoordinate_78, output=""}
    UV3 {from=MaterialExpressionTextureCoordinate_79, output=""}
   ]
   properties {inputs [
     {
      inputName=UV0
      input {
       expression=&self:MaterialExpressionTextureCoordinate_76
       inputName=None
       mask=0
       maskA=0
       maskB=0
       maskG=0
       maskR=0
       outputIndex=0
      }
     }
     {
      inputName=UV1
      input {
       expression=&self:MaterialExpressionTextureCoordinate_77
       inputName=None
       mask=0
       maskA=0
       maskB=0
       maskG=0
       maskR=0
       outputIndex=0
      }
     }
     {
      inputName=UV2
      input {
       expression=&self:MaterialExpressionTextureCoordinate_78
       inputName=None
       mask=0
       maskA=0
       maskB=0
       maskG=0
       maskR=0
       outputIndex=0
      }
     }
     {
      inputName=UV3
      input {
       expression=&self:MaterialExpressionTextureCoordinate_79
       inputName=None
       mask=0
       maskA=0
       maskB=0
       maskG=0
       maskR=0
       outputIndex=0
      }
     }
    ]}
  }
  MaterialExpressionSwitch_12 {
   class=MaterialExpressionSwitch
   inputs [
    SwitchValue {from=MaterialExpressionReroute_26, output=UVIndex}
    Default {from=MaterialExpressionTextureCoordinate_16, output=""}
    UV0 {from=MaterialExpressionTextureCoordinate_16, output=""}
    UV1 {from=MaterialExpressionTextureCoordinate_17, output=""}
    UV2 {from=MaterialExpressionTextureCoordinate_18, output=""}
    UV3 {from=MaterialExpressionTextureCoordinate_19, output=""}
   ]
   properties {inputs [
     {
      inputName=UV0
      input {
       expression=&self:MaterialExpressionTextureCoordinate_16
       inputName=None
       mask=0
       maskA=0
       maskB=0
       maskG=0
       maskR=0
       outputIndex=0
      }
     }
     {
      inputName=UV1
      input {
       expression=&self:MaterialExpressionTextureCoordinate_17
       inputName=None
       mask=0
       maskA=0
       maskB=0
       maskG=0
       maskR=0
       outputIndex=0
      }
     }
     {
      inputName=UV2
      input {
       expression=&self:MaterialExpressionTextureCoordinate_18
       inputName=None
       mask=0
       maskA=0
       maskB=0
       maskG=0
       maskR=0
       outputIndex=0
      }
     }
     {
      inputName=UV3
      input {
       expression=&self:MaterialExpressionTextureCoordinate_19
       inputName=None
       mask=0
       maskA=0
       maskB=0
       maskG=0
       maskR=0
       outputIndex=0
      }
     }
    ]}
  }
  MaterialExpressionSwitch_2 {
   class=MaterialExpressionSwitch
   inputs [
    SwitchValue {from=MaterialExpressionReroute_2, output=UVIndex}
    Default {from=MaterialExpressionTextureCoordinate_4, output=""}
    UV0 {from=MaterialExpressionTextureCoordinate_4, output=""}
    UV1 {from=MaterialExpressionTextureCoordinate_5, output=""}
    UV2 {from=MaterialExpressionTextureCoordinate_6, output=""}
    UV3 {from=MaterialExpressionTextureCoordinate_7, output=""}
   ]
   properties {inputs [
     {
      inputName=UV0
      input {
       expression=&self:MaterialExpressionTextureCoordinate_4
       inputName=None
       mask=0
       maskA=0
       maskB=0
       maskG=0
       maskR=0
       outputIndex=0
      }
     }
     {
      inputName=UV1
      input {
       expression=&self:MaterialExpressionTextureCoordinate_5
       inputName=None
       mask=0
       maskA=0
       maskB=0
       maskG=0
       maskR=0
       outputIndex=0
      }
     }
     {
      inputName=UV2
      input {
       expression=&self:MaterialExpressionTextureCoordinate_6
       inputName=None
       mask=0
       maskA=0
       maskB=0
       maskG=0
       maskR=0
       outputIndex=0
      }
     }
     {
      inputName=UV3
      input {
       expression=&self:MaterialExpressionTextureCoordinate_7
       inputName=None
       mask=0
       maskA=0
       maskB=0
       maskG=0
       maskR=0
       outputIndex=0
      }
     }
    ]}
  }
  MaterialExpressionSwitch_3 {
   class=MaterialExpressionSwitch
   inputs [
    SwitchValue {from=MaterialExpressionReroute_3, output=UVIndex}
    Default {from=MaterialExpressionTextureCoordinate_8, output=""}
    UV0 {from=MaterialExpressionTextureCoordinate_8, output=""}
    UV1 {from=MaterialExpressionTextureCoordinate_9, output=""}
    UV2 {from=MaterialExpressionTextureCoordinate_10, output=""}
    UV3 {from=MaterialExpressionTextureCoordinate_11, output=""}
   ]
   properties {inputs [
     {
      inputName=UV0
      input {
       expression=&self:MaterialExpressionTextureCoordinate_8
       inputName=None
       mask=0
       maskA=0
       maskB=0
       maskG=0
       maskR=0
       outputIndex=0
      }
     }
     {
      inputName=UV1
      input {
       expression=&self:MaterialExpressionTextureCoordinate_9
       inputName=None
       mask=0
       maskA=0
       maskB=0
       maskG=0
       maskR=0
       outputIndex=0
      }
     }
     {
      inputName=UV2
      input {
       expression=&self:MaterialExpressionTextureCoordinate_10
       inputName=None
       mask=0
       maskA=0
       maskB=0
       maskG=0
       maskR=0
       outputIndex=0
      }
     }
     {
      inputName=UV3
      input {
       expression=&self:MaterialExpressionTextureCoordinate_11
       inputName=None
       mask=0
       maskA=0
       maskB=0
       maskG=0
       maskR=0
       outputIndex=0
      }
     }
    ]}
  }
  MaterialExpressionSwitch_4 {
   class=MaterialExpressionSwitch
   inputs [
    SwitchValue {from=MaterialExpressionReroute_24, output=UVIndex}
    Default {from=MaterialExpressionTextureCoordinate_12, output=""}
    UV0 {from=MaterialExpressionTextureCoordinate_12, output=""}
    UV1 {from=MaterialExpressionTextureCoordinate_13, output=""}
    UV2 {from=MaterialExpressionTextureCoordinate_14, output=""}
    UV3 {from=MaterialExpressionTextureCoordinate_15, output=""}
   ]
   properties {inputs [
     {
      inputName=UV0
      input {
       expression=&self:MaterialExpressionTextureCoordinate_12
       inputName=None
       mask=0
       maskA=0
       maskB=0
       maskG=0
       maskR=0
       outputIndex=0
      }
     }
     {
      inputName=UV1
      input {
       expression=&self:MaterialExpressionTextureCoordinate_13
       inputName=None
       mask=0
       maskA=0
       maskB=0
       maskG=0
       maskR=0
       outputIndex=0
      }
     }
     {
      inputName=UV2
      input {
       expression=&self:MaterialExpressionTextureCoordinate_14
       inputName=None
       mask=0
       maskA=0
       maskB=0
       maskG=0
       maskR=0
       outputIndex=0
      }
     }
     {
      inputName=UV3
      input {
       expression=&self:MaterialExpressionTextureCoordinate_15
       inputName=None
       mask=0
       maskA=0
       maskB=0
       maskG=0
       maskR=0
       outputIndex=0
      }
     }
    ]}
  }
  MaterialExpressionSwitch_5 {
   class=MaterialExpressionSwitch
   inputs [
    SwitchValue {from=MaterialExpressionReroute_0, output=UVIndex}
    Default {from=MaterialExpressionTextureCoordinate_52, output=""}
    UV0 {from=MaterialExpressionTextureCoordinate_52, output=""}
    UV1 {from=MaterialExpressionTextureCoordinate_53, output=""}
    UV2 {from=MaterialExpressionTextureCoordinate_54, output=""}
    UV3 {from=MaterialExpressionTextureCoordinate_55, output=""}
   ]
   properties {inputs [
     {
      inputName=UV0
      input {
       expression=&self:MaterialExpressionTextureCoordinate_52
       inputName=None
       mask=0
       maskA=0
       maskB=0
       maskG=0
       maskR=0
       outputIndex=0
      }
     }
     {
      inputName=UV1
      input {
       expression=&self:MaterialExpressionTextureCoordinate_53
       inputName=None
       mask=0
       maskA=0
       maskB=0
       maskG=0
       maskR=0
       outputIndex=0
      }
     }
     {
      inputName=UV2
      input {
       expression=&self:MaterialExpressionTextureCoordinate_54
       inputName=None
       mask=0
       maskA=0
       maskB=0
       maskG=0
       maskR=0
       outputIndex=0
      }
     }
     {
      inputName=UV3
      input {
       expression=&self:MaterialExpressionTextureCoordinate_55
       inputName=None
       mask=0
       maskA=0
       maskB=0
       maskG=0
       maskR=0
       outputIndex=0
      }
     }
    ]}
  }
  MaterialExpressionSwitch_6 {
   class=MaterialExpressionSwitch
   inputs [
    SwitchValue {from=MaterialExpressionReroute_0, output=UVIndex}
    Default {from=MaterialExpressionTextureCoordinate_56, output=""}
    UV0 {from=MaterialExpressionTextureCoordinate_56, output=""}
    UV1 {from=MaterialExpressionTextureCoordinate_57, output=""}
    UV2 {from=MaterialExpressionTextureCoordinate_58, output=""}
    UV3 {from=MaterialExpressionTextureCoordinate_59, output=""}
   ]
   properties {inputs [
     {
      inputName=UV0
      input {
       expression=&self:MaterialExpressionTextureCoordinate_56
       inputName=None
       mask=0
       maskA=0
       maskB=0
       maskG=0
       maskR=0
       outputIndex=0
      }
     }
     {
      inputName=UV1
      input {
       expression=&self:MaterialExpressionTextureCoordinate_57
       inputName=None
       mask=0
       maskA=0
       maskB=0
       maskG=0
       maskR=0
       outputIndex=0
      }
     }
     {
      inputName=UV2
      input {
       expression=&self:MaterialExpressionTextureCoordinate_58
       inputName=None
       mask=0
       maskA=0
       maskB=0
       maskG=0
       maskR=0
       outputIndex=0
      }
     }
     {
      inputName=UV3
      input {
       expression=&self:MaterialExpressionTextureCoordinate_59
       inputName=None
       mask=0
       maskA=0
       maskB=0
       maskG=0
       maskR=0
       outputIndex=0
      }
     }
    ]}
  }
  MaterialExpressionSwitch_7 {
   class=MaterialExpressionSwitch
   inputs [
    SwitchValue {from=MaterialExpressionReroute_20, output=UVIndex}
    Default {from=MaterialExpressionTextureCoordinate_60, output=""}
    UV0 {from=MaterialExpressionTextureCoordinate_60, output=""}
    UV1 {from=MaterialExpressionTextureCoordinate_61, output=""}
    UV2 {from=MaterialExpressionTextureCoordinate_62, output=""}
    UV3 {from=MaterialExpressionTextureCoordinate_63, output=""}
   ]
   properties {inputs [
     {
      inputName=UV0
      input {
       expression=&self:MaterialExpressionTextureCoordinate_60
       inputName=None
       mask=0
       maskA=0
       maskB=0
       maskG=0
       maskR=0
       outputIndex=0
      }
     }
     {
      inputName=UV1
      input {
       expression=&self:MaterialExpressionTextureCoordinate_61
       inputName=None
       mask=0
       maskA=0
       maskB=0
       maskG=0
       maskR=0
       outputIndex=0
      }
     }
     {
      inputName=UV2
      input {
       expression=&self:MaterialExpressionTextureCoordinate_62
       inputName=None
       mask=0
       maskA=0
       maskB=0
       maskG=0
       maskR=0
       outputIndex=0
      }
     }
     {
      inputName=UV3
      input {
       expression=&self:MaterialExpressionTextureCoordinate_63
       inputName=None
       mask=0
       maskA=0
       maskB=0
       maskG=0
       maskR=0
       outputIndex=0
      }
     }
    ]}
  }
  MaterialExpressionSwitch_8 {
   class=MaterialExpressionSwitch
   inputs [
    SwitchValue {from=MaterialExpressionReroute_0, output=UVIndex}
    Default {from=MaterialExpressionTextureCoordinate_64, output=""}
    UV0 {from=MaterialExpressionTextureCoordinate_64, output=""}
    UV1 {from=MaterialExpressionTextureCoordinate_65, output=""}
    UV2 {from=MaterialExpressionTextureCoordinate_66, output=""}
    UV3 {from=MaterialExpressionTextureCoordinate_67, output=""}
   ]
   properties {inputs [
     {
      inputName=UV0
      input {
       expression=&self:MaterialExpressionTextureCoordinate_64
       inputName=None
       mask=0
       maskA=0
       maskB=0
       maskG=0
       maskR=0
       outputIndex=0
      }
     }
     {
      inputName=UV1
      input {
       expression=&self:MaterialExpressionTextureCoordinate_65
       inputName=None
       mask=0
       maskA=0
       maskB=0
       maskG=0
       maskR=0
       outputIndex=0
      }
     }
     {
      inputName=UV2
      input {
       expression=&self:MaterialExpressionTextureCoordinate_66
       inputName=None
       mask=0
       maskA=0
       maskB=0
       maskG=0
       maskR=0
       outputIndex=0
      }
     }
     {
      inputName=UV3
      input {
       expression=&self:MaterialExpressionTextureCoordinate_67
       inputName=None
       mask=0
       maskA=0
       maskB=0
       maskG=0
       maskR=0
       outputIndex=0
      }
     }
    ]}
  }
  MaterialExpressionSwitch_9 {
   class=MaterialExpressionSwitch
   inputs [
    SwitchValue {from=MaterialExpressionReroute_26, output=UVIndex}
    Default {from=MaterialExpressionTextureCoordinate_68, output=""}
    UV0 {from=MaterialExpressionTextureCoordinate_68, output=""}
    UV1 {from=MaterialExpressionTextureCoordinate_69, output=""}
    UV2 {from=MaterialExpressionTextureCoordinate_70, output=""}
    UV3 {from=MaterialExpressionTextureCoordinate_71, output=""}
   ]
   properties {inputs [
     {
      inputName=UV0
      input {
       expression=&self:MaterialExpressionTextureCoordinate_68
       inputName=None
       mask=0
       maskA=0
       maskB=0
       maskG=0
       maskR=0
       outputIndex=0
      }
     }
     {
      inputName=UV1
      input {
       expression=&self:MaterialExpressionTextureCoordinate_69
       inputName=None
       mask=0
       maskA=0
       maskB=0
       maskG=0
       maskR=0
       outputIndex=0
      }
     }
     {
      inputName=UV2
      input {
       expression=&self:MaterialExpressionTextureCoordinate_70
       inputName=None
       mask=0
       maskA=0
       maskB=0
       maskG=0
       maskR=0
       outputIndex=0
      }
     }
     {
      inputName=UV3
      input {
       expression=&self:MaterialExpressionTextureCoordinate_71
       inputName=None
       mask=0
       maskA=0
       maskB=0
       maskG=0
       maskR=0
       outputIndex=0
      }
     }
    ]}
  }
  MaterialExpressionTextureCoordinate_0 {class=MaterialExpressionTextureCoordinate}
  MaterialExpressionTextureCoordinate_1 {class=MaterialExpressionTextureCoordinate, properties {coordinateIndex=1}}
  MaterialExpressionTextureCoordinate_10 {class=MaterialExpressionTextureCoordinate, properties {coordinateIndex=2}}
  MaterialExpressionTextureCoordinate_11 {class=MaterialExpressionTextureCoordinate, properties {coordinateIndex=3}}
  MaterialExpressionTextureCoordinate_12 {class=MaterialExpressionTextureCoordinate}
  MaterialExpressionTextureCoordinate_13 {class=MaterialExpressionTextureCoordinate, properties {coordinateIndex=1}}
  MaterialExpressionTextureCoordinate_14 {class=MaterialExpressionTextureCoordinate, properties {coordinateIndex=2}}
  MaterialExpressionTextureCoordinate_15 {class=MaterialExpressionTextureCoordinate, properties {coordinateIndex=3}}
  MaterialExpressionTextureCoordinate_16 {class=MaterialExpressionTextureCoordinate}
  MaterialExpressionTextureCoordinate_17 {class=MaterialExpressionTextureCoordinate, properties {coordinateIndex=1}}
  MaterialExpressionTextureCoordinate_18 {class=MaterialExpressionTextureCoordinate, properties {coordinateIndex=2}}
  MaterialExpressionTextureCoordinate_19 {class=MaterialExpressionTextureCoordinate, properties {coordinateIndex=3}}
  MaterialExpressionTextureCoordinate_2 {class=MaterialExpressionTextureCoordinate, properties {coordinateIndex=2}}
  MaterialExpressionTextureCoordinate_3 {class=MaterialExpressionTextureCoordinate, properties {coordinateIndex=3}}
  MaterialExpressionTextureCoordinate_32 {class=MaterialExpressionTextureCoordinate}
  MaterialExpressionTextureCoordinate_33 {class=MaterialExpressionTextureCoordinate, properties {coordinateIndex=1}}
  MaterialExpressionTextureCoordinate_34 {class=MaterialExpressionTextureCoordinate, properties {coordinateIndex=2}}
  MaterialExpressionTextureCoordinate_35 {class=MaterialExpressionTextureCoordinate, properties {coordinateIndex=3}}
  MaterialExpressionTextureCoordinate_4 {class=MaterialExpressionTextureCoordinate}
  MaterialExpressionTextureCoordinate_5 {class=MaterialExpressionTextureCoordinate, properties {coordinateIndex=1}}
  MaterialExpressionTextureCoordinate_52 {class=MaterialExpressionTextureCoordinate}
  MaterialExpressionTextureCoordinate_53 {class=MaterialExpressionTextureCoordinate, properties {coordinateIndex=1}}
  MaterialExpressionTextureCoordinate_54 {class=MaterialExpressionTextureCoordinate, properties {coordinateIndex=2}}
  MaterialExpressionTextureCoordinate_55 {class=MaterialExpressionTextureCoordinate, properties {coordinateIndex=3}}
  MaterialExpressionTextureCoordinate_56 {class=MaterialExpressionTextureCoordinate}
  MaterialExpressionTextureCoordinate_57 {class=MaterialExpressionTextureCoordinate, properties {coordinateIndex=1}}
  MaterialExpressionTextureCoordinate_58 {class=MaterialExpressionTextureCoordinate, properties {coordinateIndex=2}}
  MaterialExpressionTextureCoordinate_59 {class=MaterialExpressionTextureCoordinate, properties {coordinateIndex=3}}
  MaterialExpressionTextureCoordinate_6 {class=MaterialExpressionTextureCoordinate, properties {coordinateIndex=2}}
  MaterialExpressionTextureCoordinate_60 {class=MaterialExpressionTextureCoordinate}
  MaterialExpressionTextureCoordinate_61 {class=MaterialExpressionTextureCoordinate, properties {coordinateIndex=1}}
  MaterialExpressionTextureCoordinate_62 {class=MaterialExpressionTextureCoordinate, properties {coordinateIndex=2}}
  MaterialExpressionTextureCoordinate_63 {class=MaterialExpressionTextureCoordinate, properties {coordinateIndex=3}}
  MaterialExpressionTextureCoordinate_64 {class=MaterialExpressionTextureCoordinate}
  MaterialExpressionTextureCoordinate_65 {class=MaterialExpressionTextureCoordinate, properties {coordinateIndex=1}}
  MaterialExpressionTextureCoordinate_66 {class=MaterialExpressionTextureCoordinate, properties {coordinateIndex=2}}
  MaterialExpressionTextureCoordinate_67 {class=MaterialExpressionTextureCoordinate, properties {coordinateIndex=3}}
  MaterialExpressionTextureCoordinate_68 {class=MaterialExpressionTextureCoordinate}
  MaterialExpressionTextureCoordinate_69 {class=MaterialExpressionTextureCoordinate, properties {coordinateIndex=1}}
  MaterialExpressionTextureCoordinate_7 {class=MaterialExpressionTextureCoordinate, properties {coordinateIndex=3}}
  MaterialExpressionTextureCoordinate_70 {class=MaterialExpressionTextureCoordinate, properties {coordinateIndex=2}}
  MaterialExpressionTextureCoordinate_71 {class=MaterialExpressionTextureCoordinate, properties {coordinateIndex=3}}
  MaterialExpressionTextureCoordinate_72 {class=MaterialExpressionTextureCoordinate}
  MaterialExpressionTextureCoordinate_73 {class=MaterialExpressionTextureCoordinate, properties {coordinateIndex=1}}
  MaterialExpressionTextureCoordinate_74 {class=MaterialExpressionTextureCoordinate, properties {coordinateIndex=2}}
  MaterialExpressionTextureCoordinate_75 {class=MaterialExpressionTextureCoordinate, properties {coordinateIndex=3}}
  MaterialExpressionTextureCoordinate_76 {class=MaterialExpressionTextureCoordinate}
  MaterialExpressionTextureCoordinate_77 {class=MaterialExpressionTextureCoordinate, properties {coordinateIndex=1}}
  MaterialExpressionTextureCoordinate_78 {class=MaterialExpressionTextureCoordinate, properties {coordinateIndex=2}}
  MaterialExpressionTextureCoordinate_79 {class=MaterialExpressionTextureCoordinate, properties {coordinateIndex=3}}
  MaterialExpressionTextureCoordinate_8 {class=MaterialExpressionTextureCoordinate}
  MaterialExpressionTextureCoordinate_9 {class=MaterialExpressionTextureCoordinate, properties {coordinateIndex=1}}
  MaterialExpressionTextureSampleParameter2D_0 {
   class=MaterialExpressionTextureSampleParameter2D
   inputs [
    UVs {from=MaterialExpressionMaterialFunctionCall_33, output=Result}
    Apply View MipBias {from="", output=""}
   ]
   properties {parameterName=BaseColorTexture, texture=/Engine/EditorLandscapeResources/WhiteSquareTexture}
  }
  MaterialExpressionTextureSampleParameter2D_1 {
   class=MaterialExpressionTextureSampleParameter2D
   inputs [
    UVs {from=MaterialExpressionMaterialFunctionCall_35, output=Result}
    Apply View MipBias {from="", output=""}
   ]
   properties {parameterName=MetallicTexture, samplerType=SAMPLERTYPE_LinearColor, texture=/Engine/ArtTools/RenderToTexture/Textures/127grey}
  }
  MaterialExpressionTextureSampleParameter2D_10 {
   class=MaterialExpressionTextureSampleParameter2D
   inputs [
    UVs {from=MaterialExpressionMaterialFunctionCall_47, output=Result}
    Apply View MipBias {from="", output=""}
   ]
   properties {parameterName=ClearCoatRoughnessTexture, samplerType=SAMPLERTYPE_LinearColor, texture=&127grey}
  }
  MaterialExpressionTextureSampleParameter2D_11 {
   class=MaterialExpressionTextureSampleParameter2D
   inputs [
    UVs {from=MaterialExpressionMaterialFunctionCall_48, output=Result}
    Apply View MipBias {from="", output=""}
   ]
   properties {parameterName=OpacityTexture, samplerType=SAMPLERTYPE_LinearColor, texture=&127grey}
  }
  MaterialExpressionTextureSampleParameter2D_12 {
   class=MaterialExpressionTextureSampleParameter2D
   inputs [
    UVs {from=MaterialExpressionMaterialFunctionCall_49, output=Result}
    Apply View MipBias {from="", output=""}
   ]
   properties {parameterName=DisplacementTexture, samplerType=SAMPLERTYPE_LinearColor, texture=&127grey}
  }
  MaterialExpressionTextureSampleParameter2D_2 {
   class=MaterialExpressionTextureSampleParameter2D
   inputs [
    UVs {from=MaterialExpressionMaterialFunctionCall_36, output=Result}
    Apply View MipBias {from="", output=""}
   ]
   properties {parameterName=RoughnessTexture, samplerType=SAMPLERTYPE_LinearColor, texture=&127grey}
  }
  MaterialExpressionTextureSampleParameter2D_3 {
   class=MaterialExpressionTextureSampleParameter2D
   inputs [
    UVs {from=MaterialExpressionMaterialFunctionCall_39, output=Result}
    Apply View MipBias {from="", output=""}
   ]
   properties {parameterName=OpacityThresholdTexture, samplerType=SAMPLERTYPE_LinearColor, texture=&127grey}
  }
  MaterialExpressionTextureSampleParameter2D_4 {
   class=MaterialExpressionTextureSampleParameter2D
   inputs [
    UVs {from=MaterialExpressionMaterialFunctionCall_41, output=Result}
    Apply View MipBias {from="", output=""}
   ]
   properties {parameterName=NormalTexture, samplerType=SAMPLERTYPE_Normal, texture=/Engine/EngineMaterials/BaseFlattenNormalMap}
  }
  MaterialExpressionTextureSampleParameter2D_5 {
   class=MaterialExpressionTextureSampleParameter2D
   inputs [
    UVs {from=MaterialExpressionMaterialFunctionCall_38, output=Result}
    Apply View MipBias {from="", output=""}
   ]
   properties {parameterName=EmissiveColorTexture, texture=&WhiteSquareTexture}
  }
  MaterialExpressionTextureSampleParameter2D_6 {
   class=MaterialExpressionTextureSampleParameter2D
   inputs [
    UVs {from=MaterialExpressionMaterialFunctionCall_42, output=Result}
    Apply View MipBias {from="", output=""}
   ]
   properties {parameterName=RefractionTexture, samplerType=SAMPLERTYPE_LinearColor, texture=&127grey}
  }
  MaterialExpressionTextureSampleParameter2D_7 {
   class=MaterialExpressionTextureSampleParameter2D
   inputs [
    UVs {from=MaterialExpressionMaterialFunctionCall_43, output=Result}
    Apply View MipBias {from="", output=""}
   ]
   properties {parameterName=AmbientOcclusionTexture, samplerType=SAMPLERTYPE_LinearColor, texture=&127grey}
  }
  MaterialExpressionTextureSampleParameter2D_8 {
   class=MaterialExpressionTextureSampleParameter2D
   inputs [
    UVs {from=MaterialExpressionMaterialFunctionCall_45, output=Result}
    Apply View MipBias {from="", output=""}
   ]
   properties {parameterName=SpecularColorTexture, texture=&WhiteSquareTexture}
  }
  MaterialExpressionTextureSampleParameter2D_9 {
   class=MaterialExpressionTextureSampleParameter2D
   inputs [
    UVs {from=MaterialExpressionMaterialFunctionCall_46, output=Result}
    Apply View MipBias {from="", output=""}
   ]
   properties {parameterName=ClearCoatTexture, samplerType=SAMPLERTYPE_LinearColor, texture=&127grey}
  }
  MaterialExpressionVectorParameter_0 {class=MaterialExpressionVectorParameter, properties {parameterName=BaseColor, defaultValue=<rgba 0.18 0.18 0.18 1>}}
  MaterialExpressionVectorParameter_1 {class=MaterialExpressionVectorParameter, properties {parameterName=EmissiveColor, defaultValue=<rgba 0 0 0 1>}}
  MaterialExpressionVectorParameter_10 {
   class=MaterialExpressionVectorParameter
   properties {group=ScaleTranslation, parameterName=RoughnessScaleTranslation, defaultValue=<rgba 1 1 0 0>}
  }
  MaterialExpressionVectorParameter_11 {
   class=MaterialExpressionVectorParameter
   properties {group=ScaleTranslation, parameterName=EmissiveScaleTranslation, defaultValue=<rgba 1 1 0 0>}
  }
  MaterialExpressionVectorParameter_12 {
   class=MaterialExpressionVectorParameter
   properties {group=ScaleTranslation, parameterName=OpacityThresholdScaleTranslation, defaultValue=<rgba 1 1 0 0>}
  }
  MaterialExpressionVectorParameter_13 {
   class=MaterialExpressionVectorParameter
   properties {group=ScaleTranslation, parameterName=NormalScaleTranslation, defaultValue=<rgba 1 1 0 0>}
  }
  MaterialExpressionVectorParameter_14 {
   class=MaterialExpressionVectorParameter
   properties {group=ScaleTranslation, parameterName=RefractionScaleTranslation, defaultValue=<rgba 1 1 0 0>}
  }
  MaterialExpressionVectorParameter_15 {
   class=MaterialExpressionVectorParameter
   properties {group=ScaleTranslation, parameterName=AmbientOcclusionScaleTranslation, defaultValue=<rgba 1 1 0 0>}
  }
  MaterialExpressionVectorParameter_16 {
   class=MaterialExpressionVectorParameter
   properties {group=TextureComponent, parameterName=ClearCoatTextureComponent, defaultValue=<rgba 1 0 0 0>}
  }
  MaterialExpressionVectorParameter_17 {
   class=MaterialExpressionVectorParameter
   properties {group=ScaleTranslation, parameterName=SpecularScaleTranslation, defaultValue=<rgba 1 1 0 0>}
  }
  MaterialExpressionVectorParameter_18 {class=MaterialExpressionVectorParameter, properties {parameterName=SpecularColor}}
  MaterialExpressionVectorParameter_19 {
   class=MaterialExpressionVectorParameter
   properties {group=ScaleTranslation, parameterName=ClearCoatScaleTranslation, defaultValue=<rgba 1 1 0 0>}
  }
  MaterialExpressionVectorParameter_2 {class=MaterialExpressionVectorParameter, properties {parameterName=Normal, defaultValue=<rgba 0 0 1 1>}}
  MaterialExpressionVectorParameter_20 {
   class=MaterialExpressionVectorParameter
   properties {group=TextureComponent, parameterName=ClearCoatRoughnessTextureComponent, defaultValue=<rgba 1 0 0 0>}
  }
  MaterialExpressionVectorParameter_21 {
   class=MaterialExpressionVectorParameter
   properties {group=ScaleTranslation, parameterName=ClearCoatRoughnessScaleTranslation, defaultValue=<rgba 1 1 0 0>}
  }
  MaterialExpressionVectorParameter_22 {
   class=MaterialExpressionVectorParameter
   properties {group=TextureComponent, parameterName=OpacityTextureComponent, defaultValue=<rgba 1 0 0 0>}
  }
  MaterialExpressionVectorParameter_23 {
   class=MaterialExpressionVectorParameter
   properties {group=ScaleTranslation, parameterName=OpacityScaleTranslation, defaultValue=<rgba 1 1 0 0>}
  }
  MaterialExpressionVectorParameter_24 {
   class=MaterialExpressionVectorParameter
   properties {group=TextureComponent, parameterName=DisplacementTextureComponent, defaultValue=<rgba 1 0 0 0>}
  }
  MaterialExpressionVectorParameter_25 {
   class=MaterialExpressionVectorParameter
   properties {group=ScaleTranslation, parameterName=DisplacementScaleTranslation, defaultValue=<rgba 1 1 0 0>}
  }
  MaterialExpressionVectorParameter_3 {
   class=MaterialExpressionVectorParameter
   properties {group=TextureComponent, parameterName=RefractionTextureComponent, defaultValue=<rgba 1 0 0 0>}
  }
  MaterialExpressionVectorParameter_4 {
   class=MaterialExpressionVectorParameter
   properties {group=TextureComponent, parameterName=AmbientOcclusionTextureComponent, defaultValue=<rgba 1 0 0 0>}
  }
  MaterialExpressionVectorParameter_5 {
   class=MaterialExpressionVectorParameter
   properties {group=TextureComponent, parameterName=MetallicTextureComponent, defaultValue=<rgba 1 0 0 0>}
  }
  MaterialExpressionVectorParameter_6 {
   class=MaterialExpressionVectorParameter
   properties {group=TextureComponent, parameterName=OpacityThresholdTextureComponent, defaultValue=<rgba 1 0 0 0>}
  }
  MaterialExpressionVectorParameter_7 {
   class=MaterialExpressionVectorParameter
   properties {group=TextureComponent, parameterName=RoughnessTextureComponent, defaultValue=<rgba 1 0 0 0>}
  }
  MaterialExpressionVectorParameter_8 {
   class=MaterialExpressionVectorParameter
   properties {group=ScaleTranslation, parameterName=BaseColorScaleTranslation, defaultValue=<rgba 1 1 0 0>}
  }
  MaterialExpressionVectorParameter_9 {
   class=MaterialExpressionVectorParameter
   properties {group=ScaleTranslation, parameterName=MetallicScaleTranslation, defaultValue=<rgba 1 1 0 0>}
  }
  MaterialExpressionVertexColor_0 {class=MaterialExpressionVertexColor, inputs [], properties {}}
 }
 properties {
  bUsedWithClothing=true
  bUsedWithGeometryCache=true
  bUsedWithInstancedStaticMeshes=true
  bUsedWithMeshDeformer=true
  bUsedWithMorphTargets=true
  bUsedWithNanite=true
  bUsedWithSkeletalMesh=true
  bUsedWithVoxels=true
  pixelDepthOffsetMode=PDOM_Legacy
  refractionMethod=RM_IndexOfRefraction
  shadingModel=MSM_SubsurfaceProfile
  translucencyLightingMode=TLM_SurfacePerPixelLighting
  usedShadingModels=Subsurface Profile
 }
}
